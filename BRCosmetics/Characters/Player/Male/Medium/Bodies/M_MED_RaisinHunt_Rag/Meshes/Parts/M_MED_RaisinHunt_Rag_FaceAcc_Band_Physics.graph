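# Physics asset for the headband face accessory on the male medium RaisinHunt Rag character, giving its two dangling straps physics bodies.
# Each strap is a capsule chain from a base bone, with tight angular limits at the root and looser ones further along.
# Head, neck, upper spine and left-shoulder capsules, joined by angularly locked constraints, supply the body shape.
# Only strap B's base is constrained to the head, with free rotation; strap A's base has no such constraint.
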PhysicsAsset /BRCosmetics/Characters/Player/Male/Medium/Bodies/M_MED_RaisinHunt_Rag/Meshes/Parts/M_MED_RaisinHunt_Rag_FaceAcc_Band_Physics
bodies {
 clavicle_l [{shapeType=Capsule, length=20.5117, radius=9, center=<9.9354 -0.4819 0.1782>, rotation=<rot -76.4404 179.9991 -172.7749>}]
 dyn_headband_straps_a_01 [{shapeType=Capsule, length=5, radius=1.5, shapeName=dyn_headband_straps_a_01_capsule, center=<2 0 0>, rotation=<rot -90 0 0>}]
 dyn_headband_straps_a_02 [{shapeType=Capsule, length=4.5482, radius=1.9518, shapeName=dyn_headband_straps_a_02_capsule, center=<2 0 0>, rotation=<rot -90 0 0>}]
 dyn_headband_straps_a_03 [{shapeType=Capsule, length=2.9866, radius=2.2137, shapeName=dyn_headband_straps_a_03_capsule, center=<2 0 0>, rotation=<rot -90 0 0>}]
 dyn_headband_straps_a_04 [{shapeType=Capsule, length=4.2863, radius=2.2137, shapeName=dyn_headband_straps_a_04_capsule, center=<2 0 0.6>, rotation=<rot -90 0 0>}]
 dyn_headband_straps_a_05 [{shapeType=Capsule, length=4.8746, radius=2, shapeName=dyn_headband_straps_a_05_capsule, center=<2.7779 0 1>, rotation=<rot -90 0 0>}]
 dyn_headband_straps_a_base [{shapeType=Capsule, length=5, radius=1.5, shapeName=dyn_headband_straps_a_base_capsule, center=<2 0 0>, rotation=<rot -90 0 0>}]
 dyn_headband_straps_b_01 [{shapeType=Capsule, length=4, radius=1.5, shapeName=dyn_headband_straps_b_01_capsule, center=<1 0 1>, rotation=<rot -90 0 0>}]
 dyn_headband_straps_b_02 [{shapeType=Capsule, length=4, radius=1.5, shapeName=dyn_headband_straps_b_02_capsule, center=<2 0 1>, rotation=<rot -90 0 0>}]
 dyn_headband_straps_b_03 [{shapeType=Capsule, length=4, radius=1.8, shapeName=dyn_headband_straps_b_03_capsule, center=<3 0 1>, rotation=<rot -90 0 0>}]
 dyn_headband_straps_b_04 [{shapeType=Capsule, length=2.498, radius=2, shapeName=dyn_headband_straps_b_04_capsule, center=<4 0 1>, rotation=<rot -90 0 0>}]
 dyn_headband_straps_b_base [{shapeType=Capsule, length=5, radius=1.5, shapeName=dyn_headband_straps_b_base_capsule, center=<2 0 0>, rotation=<rot -90 0 0>}]
 head [{shapeType=Capsule, length=7.1864, radius=4.6625, center=<4.0708 0 0>, rotation=<rot -90 0 0>}]
 neck_01 [{shapeType=Capsule, length=2.4813, radius=5.7334, shapeName=neck_01_capsule, rotation=<rot 0 -90 90>}]
 neck_02 [
  {shapeType=Capsule, length=3.8915, radius=5.1764, shapeName=neck_02_capsule, rotation=<rot 0 -90 90>}
  {
   shapeType=Capsule
   length=22.8521
   radius=7.1479
   shapeName=neck_02_capsule
   center=<-1.1189 -6.9101 -11.7204>
   rotation=<rot -86.4256 -0.0001 -7.6516>
  }
 ]
 spine_04 [{shapeType=Capsule, length=17.5321, radius=13, center=<5.6117 -0.3077 0>}]
 spine_05 [{shapeType=Capsule, length=18.5375, radius=10, center=<0.6155 1.9896 0>}]
 upperarm_twist_01_l [{shapeType=Capsule, length=26.1978, radius=5.879, center=<12 0 0>, rotation=<rot -90 0 0>}]
}
constraints [
 {
  bone1Name=neck_02
  bone2Name=neck_01
  swing1LimitDegrees=45
  swing1Motion=Locked
  swing2LimitDegrees=45
  swing2Motion=Locked
  twistLimitDegrees=45
  twistMotion=Locked
 }
 {
  bone1Name=neck_01
  bone2Name=spine_05
  swing1LimitDegrees=45
  swing1Motion=Locked
  swing2LimitDegrees=45
  swing2Motion=Locked
  twistLimitDegrees=45
  twistMotion=Locked
 }
 {
  bone1Name=clavicle_l
  bone2Name=spine_05
  swing1LimitDegrees=45
  swing1Motion=Locked
  swing2LimitDegrees=45
  swing2Motion=Locked
  twistLimitDegrees=45
  twistMotion=Locked
 }
 {
  bone1Name=head
  bone2Name=neck_02
  swing1LimitDegrees=45
  swing1Motion=Locked
  swing2LimitDegrees=45
  swing2Motion=Locked
  twistLimitDegrees=45
  twistMotion=Locked
 }
 {
  bone1Name=spine_05
  bone2Name=spine_04
  swing1LimitDegrees=45
  swing1Motion=Locked
  swing2LimitDegrees=45
  swing2Motion=Locked
  twistLimitDegrees=45
  twistMotion=Locked
 }
 {
  bone1Name=upperarm_twist_01_l
  bone2Name=clavicle_l
  swing1LimitDegrees=45
  swing1Motion=Locked
  swing2LimitDegrees=45
  swing2Motion=Locked
  twistLimitDegrees=45
  twistMotion=Locked
 }
 {bone1Name=dyn_headband_straps_a_01, bone2Name=dyn_headband_straps_a_base, swing1LimitDegrees=2, swing2LimitDegrees=2, twistLimitDegrees=2}
 {bone1Name=dyn_headband_straps_a_02, bone2Name=dyn_headband_straps_a_01, swing1LimitDegrees=20, swing2LimitDegrees=20, twistLimitDegrees=0}
 {bone1Name=dyn_headband_straps_a_03, bone2Name=dyn_headband_straps_a_02, swing1LimitDegrees=20, swing2LimitDegrees=20, twistLimitDegrees=0}
 {bone1Name=dyn_headband_straps_a_04, bone2Name=dyn_headband_straps_a_03, swing1LimitDegrees=40, swing2LimitDegrees=40, twistLimitDegrees=0}
 {bone1Name=dyn_headband_straps_a_05, bone2Name=dyn_headband_straps_a_04, swing1LimitDegrees=40, swing2LimitDegrees=40, twistLimitDegrees=0}
 {bone1Name=dyn_headband_straps_b_01, bone2Name=dyn_headband_straps_b_base, swing1LimitDegrees=2, swing2LimitDegrees=2, twistLimitDegrees=2}
 {bone1Name=dyn_headband_straps_b_04, bone2Name=dyn_headband_straps_b_03, swing1LimitDegrees=20, swing2LimitDegrees=20, twistLimitDegrees=2}
 {
  bone1Name=dyn_headband_straps_b_base
  bone2Name=head
  swing1LimitDegrees=45
  swing1Motion=Free
  swing2LimitDegrees=45
  swing2Motion=Free
  twistLimitDegrees=45
  twistMotion=Free
 }
 {bone1Name=dyn_headband_straps_b_02, bone2Name=dyn_headband_straps_b_01, swing1LimitDegrees=20, swing2LimitDegrees=20, twistLimitDegrees=2}
 {bone1Name=dyn_headband_straps_b_03, bone2Name=dyn_headband_straps_b_02, swing1LimitDegrees=20, swing2LimitDegrees=20, twistLimitDegrees=2}
]
