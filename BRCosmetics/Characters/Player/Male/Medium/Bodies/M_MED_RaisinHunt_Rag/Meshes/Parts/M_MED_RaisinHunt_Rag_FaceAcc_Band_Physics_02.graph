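# Physics asset for the male medium RaisinHunt Rag headband accessory, giving its two hanging straps capsule bodies for secondary motion.
# Straps A and B are bone chains with tight swing limits at the base, looser ones further along, and little or no twist.
# Head, neck, spine and left-side shoulder capsules form the body volume, mostly joined by locked angular constraints.
# Angularly free constraints link strap A's base to spine_03, B's to the head, and the straps mid-chain.
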PhysicsAsset /BRCosmetics/Characters/Player/Male/Medium/Bodies/M_MED_RaisinHunt_Rag/Meshes/Parts/M_MED_RaisinHunt_Rag_FaceAcc_Band_Physics_02
bodies {
 clavicle_l [{shapeType=Capsule, length=13.0838, radius=7, center=<7.6834 -0.1067 0.3325>, rotation=<rot -79.6318 -105.4297 108.6414>}]
 deform_upperarm_up_l [{
   shapeType=Capsule
   length=11.6716
   radius=6.2543
   shapeName=deform_upperarm_up_l_capsule
   center=<0 1.3325 0>
   rotation=<rot -71.3653 0.0002 -0.0001>
  }]
 dyn_headband_straps_a_01 [{shapeType=Capsule, length=5, radius=1.5, shapeName=dyn_headband_straps_a_01_capsule, center=<2 0 0>, rotation=<rot -90 0 0>}]
 dyn_headband_straps_a_02 [{shapeType=Capsule, length=5.7115, radius=1.9518, shapeName=dyn_headband_straps_a_02_capsule, center=<2 0 0.9067>, rotation=<rot -90 0 0>}]
 dyn_headband_straps_a_03 [{
   shapeType=Capsule
   length=5.422
   radius=2.2137
   shapeName=dyn_headband_straps_a_03_capsule
   center=<4 0.58 2>
   rotation=<rot -89.2595 4.5787 -0.042>
  }]
 dyn_headband_straps_a_04 [{shapeType=Capsule, length=4.2863, radius=2.1, shapeName=dyn_headband_straps_a_04_capsule, center=<2 0 2>, rotation=<rot -90 0 0>}]
 dyn_headband_straps_a_05 [{shapeType=Capsule, length=4.8108, radius=1.9, shapeName=dyn_headband_straps_a_05_capsule, center=<2.7779 0 1.5>, rotation=<rot -90 0 0>}]
 dyn_headband_straps_a_base [{shapeType=Capsule, length=5, radius=1.5, shapeName=dyn_headband_straps_a_base_capsule, center=<2 0 0>, rotation=<rot -90 0 0>}]
 dyn_headband_straps_b_01 [{shapeType=Capsule, length=4, radius=1.5, shapeName=dyn_headband_straps_b_01_capsule, center=<1 0 1>, rotation=<rot -90 0 0>}]
 dyn_headband_straps_b_02 [{shapeType=Capsule, length=4, radius=1.8, shapeName=dyn_headband_straps_b_03_capsule, center=<1.8551 0 1>, rotation=<rot -33.1953 90 -90>}]
 dyn_headband_straps_b_03 [{shapeType=Capsule, length=4, radius=1.8, shapeName=dyn_headband_straps_b_03_capsule, center=<3 0 1>, rotation=<rot -33.1953 90 -90>}]
 dyn_headband_straps_b_04 [{shapeType=Capsule, length=2.498, radius=2, shapeName=dyn_headband_straps_b_04_capsule, center=<4 0 1.5>, rotation=<rot -28.6634 90 -90>}]
 dyn_headband_straps_b_base [{shapeType=Capsule, length=5, radius=1.5, shapeName=dyn_headband_straps_b_base_capsule, center=<2 0 0>, rotation=<rot -90 0 0>}]
 head [{shapeType=Capsule, length=8.1729, radius=5.256, center=<3.1517 -0.2778 -0.2596>, rotation=<rot -90 0 0>}]
 neck_01 [{shapeType=Capsule, length=2.4813, radius=5.7334, shapeName=neck_01_capsule, rotation=<rot 0 -90 90>}]
 neck_02 [{shapeType=Capsule, length=6.1561, radius=5.3951, shapeName=neck_02_capsule, rotation=<rot 0 -90 90>}]
 spine_03 [{
   shapeType=Capsule
   length=9.123
   radius=9
   shapeName=spine_03_capsule
   center=<15.0883 2.5353 -7.5454>
   rotation=<rot -0.1162 -116.769 94.8743>
  }]
 spine_04 [{shapeType=Capsule, length=15.4261, radius=12.5, center=<5.6117 -0.3077 0>}]
 spine_05 [{shapeType=Capsule, length=15, radius=10, center=<0.4389 0.4606 -3.4479>}]
 upperarm_twist_01_l [{shapeType=Capsule, length=26.1978, radius=6.5, center=<12 0 0>, rotation=<rot -90 0 0>}]
}
constraints [
 {
  bone1Name=neck_02
  bone2Name=neck_01
  swing1LimitDegrees=45
  swing1Motion=Locked
  swing2LimitDegrees=45
  swing2Motion=Locked
  twistLimitDegrees=45
  twistMotion=Locked
 }
 {
  bone1Name=neck_01
  bone2Name=spine_05
  swing1LimitDegrees=45
  swing1Motion=Locked
  swing2LimitDegrees=45
  swing2Motion=Locked
  twistLimitDegrees=45
  twistMotion=Locked
 }
 {
  bone1Name=clavicle_l
  bone2Name=spine_05
  swing1LimitDegrees=45
  swing1Motion=Locked
  swing2LimitDegrees=45
  swing2Motion=Locked
  twistLimitDegrees=45
  twistMotion=Locked
 }
 {
  bone1Name=head
  bone2Name=neck_02
  swing1LimitDegrees=45
  swing1Motion=Locked
  swing2LimitDegrees=45
  swing2Motion=Locked
  twistLimitDegrees=45
  twistMotion=Locked
 }
 {
  bone1Name=spine_05
  bone2Name=spine_04
  swing1LimitDegrees=45
  swing1Motion=Locked
  swing2LimitDegrees=45
  swing2Motion=Locked
  twistLimitDegrees=45
  twistMotion=Locked
 }
 {
  bone1Name=upperarm_twist_01_l
  bone2Name=clavicle_l
  swing1LimitDegrees=45
  swing1Motion=Locked
  swing2LimitDegrees=45
  swing2Motion=Locked
  twistLimitDegrees=45
  twistMotion=Locked
 }
 {bone1Name=dyn_headband_straps_a_01, bone2Name=dyn_headband_straps_a_base, swing1LimitDegrees=2, swing2LimitDegrees=2, twistLimitDegrees=2}
 {bone1Name=dyn_headband_straps_a_02, bone2Name=dyn_headband_straps_a_01, swing1LimitDegrees=20, swing2LimitDegrees=20, twistLimitDegrees=0}
 {bone1Name=dyn_headband_straps_a_03, bone2Name=dyn_headband_straps_a_02, swing1LimitDegrees=20, swing2LimitDegrees=20, twistLimitDegrees=0}
 {bone1Name=dyn_headband_straps_a_04, bone2Name=dyn_headband_straps_a_03, swing1LimitDegrees=40, swing2LimitDegrees=40, twistLimitDegrees=0}
 {bone1Name=dyn_headband_straps_a_05, bone2Name=dyn_headband_straps_a_04, swing1LimitDegrees=40, swing2LimitDegrees=40, twistLimitDegrees=0}
 {bone1Name=dyn_headband_straps_b_01, bone2Name=dyn_headband_straps_b_base, swing1LimitDegrees=5, swing2LimitDegrees=5, twistLimitDegrees=2}
 {bone1Name=dyn_headband_straps_b_04, bone2Name=dyn_headband_straps_b_03, swing1LimitDegrees=20, swing2LimitDegrees=20, twistLimitDegrees=2}
 {
  bone1Name=dyn_headband_straps_b_base
  bone2Name=head
  swing1LimitDegrees=45
  swing1Motion=Free
  swing2LimitDegrees=45
  swing2Motion=Free
  twistLimitDegrees=45
  twistMotion=Free
 }
 {bone1Name=dyn_headband_straps_b_02, bone2Name=dyn_headband_straps_b_01, swing1LimitDegrees=20, swing2LimitDegrees=20, twistLimitDegrees=2}
 {bone1Name=dyn_headband_straps_b_03, bone2Name=dyn_headband_straps_b_02, swing1LimitDegrees=20, swing2LimitDegrees=20, twistLimitDegrees=2}
 {
  bone1Name=deform_upperarm_up_l
  bone2Name=upperarm_twist_01_l
  swing1LimitDegrees=45
  swing1Motion=Locked
  swing2LimitDegrees=45
  swing2Motion=Locked
  twistLimitDegrees=45
  twistMotion=Locked
 }
 {
  bone1Name=dyn_headband_straps_a_03
  bone2Name=dyn_headband_straps_b_02
  swing1LimitDegrees=45
  swing1Motion=Free
  swing2LimitDegrees=45
  swing2Motion=Free
  twistLimitDegrees=45
  twistMotion=Free
 }
 {
  bone1Name=spine_04
  bone2Name=spine_03
  swing1LimitDegrees=45
  swing1Motion=Free
  swing2LimitDegrees=45
  swing2Motion=Free
  twistLimitDegrees=45
  twistMotion=Free
 }
 {
  bone1Name=dyn_headband_straps_a_base
  bone2Name=spine_03
  swing1LimitDegrees=45
  swing1Motion=Free
  swing2LimitDegrees=45
  swing2Motion=Free
  twistLimitDegrees=45
  twistMotion=Free
 }
]
