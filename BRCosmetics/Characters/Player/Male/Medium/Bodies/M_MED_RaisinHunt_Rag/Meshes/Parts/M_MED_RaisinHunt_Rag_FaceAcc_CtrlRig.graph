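# Control Rig for the face accessory of the male medium RaisinHunt_Rag body cosmetic, posing hair and headband strap bones as the head moves.
# A spherical pose reader on the head drives the weight of corrective transform offsets on the left, middle and right hair chains.
# Parent constraints then blend both strap chains from head-driven roots toward neck, spine, left clavicle and pec bones.
# Two DeltaRotation swing/twist-to-weight functions exist in the library but go unused.
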ControlRigBlueprint /BRCosmetics/Characters/Player/Male/Medium/Bodies/M_MED_RaisinHunt_Rag/Meshes/Parts/M_MED_RaisinHunt_Rag_FaceAcc_CtrlRig
graphs [
 New Function_ContainedGraph {nodes [
   n1 = RigVMFunctionEntryNode()
   RigVMFunction_MathBoolToInteger = &RigVMFunction_MathBoolToInteger(Value=@n1.onlyTwist)
   RigVMFunction_MathQuaternionSwingTwist = &RigVMFunction_MathQuaternionSwingTwist(
    twistAxis=@n1.twistAxis
    Input=&RigUnit_GetRelativeTransformForItem(bChildInitial=False, bParentInitial=False, child=@n1.child, parent=@n1.parent).RelativeTransform.Rotation
   )
   n2 = RigVMDispatchNode(
    Values=((),())
    Index=@RigVMFunction_MathBoolToInteger.Result
    Values.0=@RigVMFunction_MathQuaternionSwingTwist.Swing
    Values.1=@RigVMFunction_MathQuaternionSwingTwist.Twist
   ) {template="DISPATCH_RigVMDispatch_SelectInt32(in Index,in Values,out Result)"}
   RigVMFunction_MathQuaternionToEuler = &RigVMFunction_MathQuaternionToEuler(Value=@n2.Result, RotationOrder=@n1.Rotation Order)
   > DISPATCH_RigVMDispatch_Print = RigVMDispatchNode(ScreenColor=(R=1,G=1,B=1,A=1), ScreenDuration=1.5, Enabled=@n1.debugRotation, Value=@RigVMFunction_MathQuaternionToEuler.Result) {template="DISPATCH_RigVMDispatch_Print(in Prefix,in Value,in Enabled,in ScreenDuration,in ScreenColor)"}
   Cast = RigVMDispatchNode(Value=@n1.triggerAxis) {template="DISPATCH_RigVMDispatch_CastEnumToInt(in Value,out Result)"}
   DISPATCH_RigVMDispatch_SelectInt32 = RigVMDispatchNode(
    Values=(0,0,0)
    Index=@Cast.Result
    Values.0=@RigVMFunction_MathQuaternionToEuler.Result.X
    Values.1=@RigVMFunction_MathQuaternionToEuler.Result.Y
    Values.2=@RigVMFunction_MathQuaternionToEuler.Result.Z
   ) {template="DISPATCH_RigVMDispatch_SelectInt32(in Index,in Values,out Result)"}
   Remap = &RigVMFunction_MathDoubleRemap(
    bClamp=False
    Value=@DISPATCH_RigVMDispatch_SelectInt32.Result
    TargetMinimum=@n1.outputMinMax.X
    TargetMaximum=@n1.outputMinMax.Y
    SourceMinimum=@n1.rotationMinMax.X
    SourceMaximum=@n1.rotationMinMax.Y
   )
   Absolute = &RigVMFunction_MathDoubleAbs(Value=@Remap.Result)
   Multiply = &RigVMFunction_MathDoubleMul(
    A=&RigVMFunction_MathDoubleClamp(Value=RigVMDispatchNode(True=@Absolute.Result, Condition=@n1.absoluteValue, False=@Remap.Result) {template="DISPATCH_RigVMDispatch_If(in Condition,in True,in False,out Result)"}.Result).Result
    B=@n1.premult
   )
   > RigVMFunctionReturnNode(outputQuat=@n2.Result, outputWeight=@Multiply.Result)
  ]}
 New Function_ContainedGraph {nodes [
   n1 = RigVMFunctionEntryNode()
   RigVMFunction_MathBoolToInteger = &RigVMFunction_MathBoolToInteger(Value=@n1.onlyTwist)
   RigVMFunction_MathQuaternionSwingTwist = &RigVMFunction_MathQuaternionSwingTwist(
    twistAxis=@n1.twistAxis
    Input=&RigUnit_GetRelativeTransformForItem(bChildInitial=False, bParentInitial=False, child=@n1.child, parent=@n1.parent).RelativeTransform.Rotation
   )
   n2 = RigVMDispatchNode(
    Values=((),())
    Index=@RigVMFunction_MathBoolToInteger.Result
    Values.0=@RigVMFunction_MathQuaternionSwingTwist.Swing
    Values.1=@RigVMFunction_MathQuaternionSwingTwist.Twist
   ) {template="DISPATCH_RigVMDispatch_SelectInt32(in Index,in Values,out Result)"}
   RigVMFunction_MathQuaternionToEuler = &RigVMFunction_MathQuaternionToEuler(Value=@n2.Result, RotationOrder=@n1.Rotation Order)
   > DISPATCH_RigVMDispatch_Print = RigVMDispatchNode(ScreenColor=(R=1,G=1,B=1,A=1), ScreenDuration=1.5, Enabled=@n1.debugRotation, Value=@RigVMFunction_MathQuaternionToEuler.Result) {template="DISPATCH_RigVMDispatch_Print(in Prefix,in Value,in Enabled,in ScreenDuration,in ScreenColor)"}
   Cast = RigVMDispatchNode(Value=@n1.triggerAxis) {template="DISPATCH_RigVMDispatch_CastEnumToInt(in Value,out Result)"}
   DISPATCH_RigVMDispatch_SelectInt32 = RigVMDispatchNode(
    Values=(0,0,0)
    Index=@Cast.Result
    Values.0=@RigVMFunction_MathQuaternionToEuler.Result.X
    Values.1=@RigVMFunction_MathQuaternionToEuler.Result.Y
    Values.2=@RigVMFunction_MathQuaternionToEuler.Result.Z
   ) {template="DISPATCH_RigVMDispatch_SelectInt32(in Index,in Values,out Result)"}
   Remap = &RigVMFunction_MathDoubleRemap(
    bClamp=False
    Value=@DISPATCH_RigVMDispatch_SelectInt32.Result
    TargetMinimum=@n1.outputMinMax.X
    TargetMaximum=@n1.outputMinMax.Y
    SourceMinimum=@n1.rotationMinMax.X
    SourceMaximum=@n1.rotationMinMax.Y
   )
   Absolute = &RigVMFunction_MathDoubleAbs(Value=@Remap.Result)
   Multiply = &RigVMFunction_MathDoubleMul(
    A=&RigVMFunction_MathDoubleClamp(Value=RigVMDispatchNode(True=@Absolute.Result, Condition=@n1.absoluteValue, False=@Remap.Result) {template="DISPATCH_RigVMDispatch_If(in Condition,in True,in False,out Result)"}.Result).Result
    B=@n1.premult
   )
   > RigVMFunctionReturnNode(outputQuat=@n2.Result, outputWeight=@Multiply.Result)
  ]}
 RigVMFunctionLibrary {nodes [
   DeltaRotation = RigVMCollapseNode(
    ExecuteContext=()
    Rotation Order=ZYX
    child=(Type=Bone,Name="None")
    outputMinMax=(X=0,Y=1)
    parent=(Type=Bone,Name="None")
    premult=1
    rotationMinMax=(X=0,Y=0)
    twistAxis=(X=1,Y=0,Z=0)
   )
   DeltaRotation v4 = RigVMCollapseNode(
    ExecuteContext=()
    Rotation Order=ZYX
    child=(Type=Bone,Name="None")
    outputMinMax=(X=0,Y=1)
    parent=(Type=Bone,Name="None")
    premult=1
    rotationMinMax=(X=0,Y=0)
    twistAxis=(X=1,Y=0,Z=0)
   )
  ]}
 RigVMModel {nodes [
   RigUnit_BeginExecution = &RigUnit_BeginExecution()
   n1 = &RigUnit_SphericalPoseReader(
    ActiveRegionScaleFactors=(PositiveWidth=1,NegativeWidth=1,PositiveHeight=1,NegativeHeight=1)
    ActiveRegionSize=0.1
    Debug=(bDrawDebug=false,bDraw2D=false,bDrawLocalAxes=False,DebugScale=25,DebugSegments=20,DebugThickness=0.25)
    DriverAxis=(X=1,Y=0,Z=0)
    DriverItem=(Type=Bone,Name="head")
    FalloffRegionScaleFactors=(PositiveWidth=0.5,NegativeWidth=0.4,PositiveHeight=1,NegativeHeight=1)
    FalloffSize=0.5
    FlipHeightScaling=False
    FlipWidthScaling=False
    OptionalParentItem=(Type=Bone,Name="None")
    RotationOffset=(X=90,Y=0,Z=-135)
    ExecutePin=@RigUnit_BeginExecution.ExecutePin
   )
   > &RigUnit_ModifyTransforms(
    ItemToModify="((Item=(Type=Bone,Name=\"dyn_left_hair_a_base\"),Transform=(Rotation=(X=-0.1447,Y=-0.2669,Z=-0.4402,W=0.845),Translation=(X=-3.3673,Y=-2.3391,Z=0.0389),Scale3D=(X=1,Y=1,Z=1))),(Item=(Type=Bone,Name=\"dyn_left_hair_a_02\"),Transform=(Rotation=(X=0,Y=0,Z=0,W=1), ...(truncated)"
    WeightMinimum=0
    Weight=@n1.OutputParam
   )
   > &RigUnit_ModifyTransforms(
    ItemToModify="((Item=(Type=Bone,Name=\"None\"),Transform=(Rotation=(X=0,Y=0,Z=0,W=1),Translation=(X=0,Y=0,Z=0),Scale3D=(X=1,Y=1,Z=1))),(Item=(Type=Bone,Name=\"dyn_left_hair_base\"),Transform=(Rotation=(X=-0.0236,Y=-0.2959,Z=-0.183,W=0.9372),Translation=(X=0,Y=0,Z=0),Scale3D ...(truncated)"
    WeightMinimum=0
    Weight=@n1.OutputParam
   )
   > &RigUnit_ModifyTransforms(
    ItemToModify="((Item=(Type=Bone,Name=\"None\"),Transform=(Rotation=(X=0,Y=0,Z=0,W=1),Translation=(X=0,Y=0,Z=0),Scale3D=(X=1,Y=1,Z=1))),(Item=(Type=Bone,Name=\"dyn_middle_hair_base\"),Transform=(Rotation=(X=-0.0042,Y=0.5451,Z=0.013,W=-0.8383),Translation=(X=0,Y=0,Z=0),Scale3 ...(truncated)"
    WeightMinimum=0
    Weight=@n1.OutputParam
   )
   > &RigUnit_ModifyTransforms(
    ItemToModify="((Item=(Type=Bone,Name=\"None\"),Transform=(Rotation=(X=0,Y=0,Z=0,W=1),Translation=(X=0,Y=0,Z=0),Scale3D=(X=1,Y=1,Z=1))),(Item=(Type=Bone,Name=\"dyn_right_hair_base\"),Transform=(Rotation=(X=-0.2681,Y=-0.545,Z=0.0594,W=0.7922),Translation=(X=-4.2829,Y=1.4849,Z ...(truncated)"
    WeightMinimum=0
    Weight=@n1.OutputParam
   )
   > &RigUnit_ModifyTransforms(
    ItemToModify="((Item=(Type=Bone,Name=\"None\"),Transform=(Rotation=(X=0,Y=0,Z=0,W=1),Translation=(X=0,Y=0,Z=0),Scale3D=(X=1,Y=1,Z=1))),(Item=(Type=Bone,Name=\"dyn_right_hair_a_base\"),Transform=(Rotation=(X=0.1893,Y=-0.3564,Z=0.5807,W=0.707),Translation=(X=-2.4909,Y=0.6612, ...(truncated)"
    WeightMinimum=0
    Weight=@n1.OutputParam
   )
   > &RigUnit_ModifyTransforms(
    ItemToModify=((Item=(Type=Bone,Name="dyn_headband_straps_a_02"),Transform=(Rotation=(X=0,Y=0,Z=0,W=1),Translation=(X=0,Y=0,Z=0),Scale3D=(X=1,Y=1,Z=1))))
    Weight=1
    WeightMinimum=0
   )
   > &RigUnit_ParentConstraint(
    AdvancedSettings=(RotationOrderForFilter=XZY)
    Child=(Type=Bone,Name="dyn_headband_straps_a_01")
    Parents=((Item=(Type=Bone,Name="head"),Weight=1.0))
    Weight=1
   )
   > &RigUnit_ParentConstraint(
    AdvancedSettings=(RotationOrderForFilter=XZY)
    Child=(Type=Bone,Name="dyn_headband_straps_a_02")
    Parents=((Item=(Type=Bone,Name="neck_02"),Weight=0.03),(Item=(Type=Bone,Name="head"),Weight=0.97))
    Weight=1
   )
   > &RigUnit_ParentConstraint(
    AdvancedSettings=(RotationOrderForFilter=XZY)
    Child=(Type=Bone,Name="dyn_headband_straps_a_03")
    Parents="((Item=(Type=Bone,Name=\"spine_03\"),Weight=0.0003),(Item=(Type=Bone,Name=\"deform_pec_l\"),Weight=0.0004),(Item=(Type=Bone,Name=\"spine_04\"),Weight=0.0012),(Item=(Type=Bone,Name=\"clavicle_l\"),Weight=0.017),(Item=(Type=Bone,Name=\"neck_01\"),Weight=0.0463),(Item= ...(truncated)"
    Weight=1
   )
   > &RigUnit_ParentConstraint(
    AdvancedSettings=(RotationOrderForFilter=XZY)
    Child=(Type=Bone,Name="dyn_headband_straps_a_04")
    Parents="((Item=(Type=Bone,Name=\"spine_03\"),Weight=0.0103),(Item=(Type=Bone,Name=\"head\"),Weight=0.0124),(Item=(Type=Bone,Name=\"deform_pec_l\"),Weight=0.0146),(Item=(Type=Bone,Name=\"neck_01\"),Weight=0.0147),(Item=(Type=Bone,Name=\"neck_02\"),Weight=0.0198),(Item=(Type= ...(truncated)"
    Weight=1
   )
   > &RigUnit_ParentConstraint(
    AdvancedSettings=(RotationOrderForFilter=XZY)
    Child=(Type=Bone,Name="dyn_headband_straps_a_05")
    Parents=((Item=(Type=Bone,Name="spine_03"),Weight=0.066),(Item=(Type=Bone,Name="deform_pec_l"),Weight=0.1538),(Item=(Type=Bone,Name="spine_04"),Weight=0.1734),(Item=(Type=Bone,Name="clavicle_l"),Weight=0.2653),(Item=(Type=Bone,Name="spine_05"),Weight=0.3415))
    Weight=1
   )
   > &RigUnit_ParentConstraint(
    AdvancedSettings=(RotationOrderForFilter=XZY)
    Child=(Type=Bone,Name="dyn_headband_straps_b_01")
    Parents=((Item=(Type=Bone,Name="head"),Weight=1.0))
    Weight=1
   )
   > &RigUnit_ParentConstraint(
    AdvancedSettings=(RotationOrderForFilter=XZY)
    Child=(Type=Bone,Name="dyn_headband_straps_b_02")
    Parents=((Item=(Type=Bone,Name="head"),Weight=1.0))
    Weight=1
   )
   > &RigUnit_ParentConstraint(
    AdvancedSettings=(RotationOrderForFilter=XZY)
    Child=(Type=Bone,Name="dyn_headband_straps_b_03")
    Parents=((Item=(Type=Bone,Name="clavicle_l"),Weight=0.0002),(Item=(Type=Bone,Name="neck_01"),Weight=0.0866),(Item=(Type=Bone,Name="head"),Weight=0.2643),(Item=(Type=Bone,Name="neck_02"),Weight=0.3169),(Item=(Type=Bone,Name="spine_05"),Weight=0.332))
    Weight=1
   )
   > &RigUnit_ParentConstraint(
    AdvancedSettings=(RotationOrderForFilter=XZY)
    Child=(Type=Bone,Name="dyn_headband_straps_b_04")
    Parents="((Item=(Type=Bone,Name=\"spine_04\"),Weight=0.0062),(Item=(Type=Bone,Name=\"head\"),Weight=0.0075),(Item=(Type=Bone,Name=\"neck_01\"),Weight=0.038),(Item=(Type=Bone,Name=\"neck_02\"),Weight=0.0989),(Item=(Type=Bone,Name=\"clavicle_l\"),Weight=0.2023),(Item=(Type=Bon ...(truncated)"
    Weight=1
   )
  ]}
]
hierarchy [
 root {loc=<0 0 0>, rot=<0 0 0 1>}
 pelvis {parent=root, loc=<0 -0.7071 93.1191>, rot=<0.0127 0.707 -0.0127 -0.707>}
 spine_01 {parent=pelvis, loc=<6.7494 -0.6388 0>, rot=<0 0 0.034 -0.9994>}
 spine_02 {parent=spine_01, loc=<7.2518 -0.035 0>, rot=<0 0 0.0238 -0.9997>}
 spine_03 {parent=spine_02, loc=<6.0104 0 0>, rot=<0 0 0.0149 -0.9999>}
 spine_04 {parent=spine_03, loc=<8.1024 0 0>, rot=<0 0 -0.0132 -0.9999>}
 spine_05 {parent=spine_04, loc=<10.0146 0 0>, rot=<0 0 -0.0015 -1>}
 neck_01 {parent=spine_05, loc=<11.3138 0.1529 0>, rot=<0 0 -0.2003 -0.9797>}
 neck_02 {parent=neck_01, loc=<4.5627 0 0>, rot=<0 0 0.0504 -0.9987>}
 head {parent=neck_02, loc=<5.3914 0 0>, rot=<0 0 0.115 -0.9934>}
 dyn_headband_straps_a_base {parent=head, loc=<7.0799 1.4354 -6.8846>, rot=<0.0411 -0.9991 -0.0004 -0.0102>}
 dyn_headband_straps_a_01 {parent=dyn_headband_straps_a_base, loc=<0 0 0>, rot=<0 0 0 1>}
 dyn_headband_straps_a_02 {parent=dyn_headband_straps_a_01, loc=<6 0 0>, rot=<0.0007 0.0217 -0.0304 -0.9993>}
 dyn_headband_straps_a_03 {parent=dyn_headband_straps_a_02, loc=<6 0 0>, rot=<0.1993 -0.0037 -0.2715 -0.9416>}
 dyn_headband_straps_a_04 {parent=dyn_headband_straps_a_03, loc=<6 0 0>, rot=<0.2434 0.0652 0.0036 -0.9677>}
 dyn_headband_straps_a_05 {parent=dyn_headband_straps_a_04, loc=<6 0 0>, rot=<0.1442 0 0 -0.9895>}
 dyn_right_hair_c_base {parent=head, loc=<6.781 1.9354 7.702>, rot=<0 0 1 0.009>}
 dyn_right_hair_c_01 {parent=dyn_right_hair_c_base, loc=<0 0 0>, rot=<0 0 0 1>}
 dyn_front_r_hair_base {parent=head, loc=<10.6966 9.0808 3.4445>, rot=<0.0064 0.7071 0.7071 0.0064>}
 dyn_front_r_hair_01 {parent=dyn_front_r_hair_base, loc=<0 0 0>, rot=<0 0 0 1>}
 dyn_front_l_hair_base {parent=head, loc=<10.6021 9.2978 -0.7976>, rot=<0.0064 0.7071 0.7071 0.0064>}
 dyn_front_l_hair_01 {parent=dyn_front_l_hair_base, loc=<0 0 0>, rot=<0 0 0 1>}
 dyn_left_hair_c_base {parent=head, loc=<9.6876 4.4139 -6.7563>, rot=<-0.009 -1 0 0>}
 dyn_left_hair_c_01 {parent=dyn_left_hair_c_base, loc=<0 0 0>, rot=<0 0 0 1>}
 dyn_back_hair_a_base {parent=head, loc=<9.1651 -8.1458 -3.626>, rot=<-0.0068 -0.7515 0.6596 0.006>}
 dyn_back_hair_a_01 {parent=dyn_back_hair_a_base, loc=<0 0 0>, rot=<0 0 0 1>}
 dyn_back_hair_b_base {parent=head, loc=<9.1957 -8.4865 1.4467>, rot=<0.0317 -0.7064 0.7064 -0.0317>}
 dyn_back_hair_b_01 {parent=dyn_back_hair_b_base, loc=<0 0 0>, rot=<0 0 0 1>}
 dyn_left_hair_b_base {parent=head, loc=<9.1216 -5.7876 -7.1321>, rot=<0.0168 -0.9616 0.2695 -0.0495>}
 dyn_left_hair_b_01 {parent=dyn_left_hair_b_base, loc=<0 0 0>, rot=<0 0 0 1>}
 dyn_middle_hair_base {parent=head, loc=<5.1887 -7.3721 0>, rot=<0.1132 -0.698 0.698 -0.1132>}
 dyn_middle_hair_01 {parent=dyn_middle_hair_base, loc=<0 0 0>, rot=<0 0 0 1>}
 dyn_middle_hair_02 {parent=dyn_middle_hair_01, loc=<6 0 0>, rot=<0 0 0 1>}
 dyn_middle_hair_03 {parent=dyn_middle_hair_02, loc=<6 0 0>, rot=<0 0 0 1>}
 dyn_middle_hair_04 {parent=dyn_middle_hair_03, loc=<6 0 0>, rot=<0 0 0 1>}
 dyn_right_hair_base {parent=head, loc=<5.0257 -6.9089 5.3399>, rot=<0.1149 -0.5761 0.8012 -0.1143>}
 dyn_right_hair_01 {parent=dyn_right_hair_base, loc=<0 0 0>, rot=<0 0 0 1>}
 dyn_right_hair_02 {parent=dyn_right_hair_01, loc=<6 0 0>, rot=<0 0 0 1>}
 dyn_right_hair_03 {parent=dyn_right_hair_02, loc=<6 0 0>, rot=<0 0 0 1>}
 dyn_right_hair_04 {parent=dyn_right_hair_03, loc=<6 0 0>, rot=<0 0 0 1>}
 dyn_right_hair_a_base {parent=head, loc=<5.9634 -3.4254 7.7828>, rot=<0.0322 -0.1771 0.9798 -0.0875>}
 dyn_right_hair_a_01 {parent=dyn_right_hair_a_base, loc=<0 0 0>, rot=<0 0 0 1>}
 dyn_right_hair_a_02 {parent=dyn_right_hair_a_01, loc=<6 0 0>, rot=<0 0 0 1>}
 dyn_right_hair_a_03 {parent=dyn_right_hair_a_02, loc=<6 0 0>, rot=<0 0 0 1>}
 dyn_right_hair_a_04 {parent=dyn_right_hair_a_03, loc=<6 0 0>, rot=<0 0 0 1>}
 dyn_right_hair_a_05 {parent=dyn_right_hair_a_02, loc=<1.4226 -0.8845 -1.04>, rot=<-0.1975 -0.0807 0.3247 -0.9214>}
 dyn_left_hair_a_base {parent=head, loc=<5.6938 -3.0382 -7.9307>, rot=<0.0991 -0.9834 0.1494 -0.0269>}
 dyn_left_hair_a_01 {parent=dyn_left_hair_a_base, loc=<0 0 0>, rot=<0 0 0 1>}
 dyn_left_hair_a_02 {parent=dyn_left_hair_a_01, loc=<6 0 0>, rot=<0 0 0 1>}
 dyn_left_hair_a_03 {parent=dyn_left_hair_a_02, loc=<6 0 0>, rot=<0 0 0 1>}
 dyn_left_hair_a_04 {parent=dyn_left_hair_a_03, loc=<6 0 0>, rot=<0 0 0 1>}
 dyn_left_hair_base {parent=head, loc=<5.0117 -6.8292 -5.1299>, rot=<0.1201 -0.8696 0.4667 -0.1071>}
 dyn_left_hair_01 {parent=dyn_left_hair_base, loc=<0 0 0>, rot=<0 0 0 1>}
 dyn_left_hair_02 {parent=dyn_left_hair_01, loc=<6 0 0>, rot=<0 0 0 1>}
 dyn_left_hair_03 {parent=dyn_left_hair_02, loc=<6 0 0>, rot=<0 0 0 1>}
 dyn_left_hair_04 {parent=dyn_left_hair_03, loc=<6 0 0>, rot=<0 0 0 1>}
 dyn_right_hair_b_base {parent=head, loc=<9.2987 -3.6621 7.8853>, rot=<-0.0048 -0.2179 0.9658 -0.1409>}
 dyn_right_hair_b_01 {parent=dyn_right_hair_b_base, loc=<0 0 0>, rot=<0 0 0 1>}
 dyn_headband_straps_b_base {parent=head, loc=<7.091 1.3015 -6.8408>, rot=<0.0411 -0.9989 0.0009 0.0225>}
 dyn_headband_straps_b_01 {parent=dyn_headband_straps_b_base, loc=<0 0 0>, rot=<0 0 0 1>}
 dyn_headband_straps_b_02 {parent=dyn_headband_straps_b_01, loc=<6 0 0>, rot=<0.0352 -0.0211 0.1028 -0.9939>}
 dyn_headband_straps_b_03 {parent=dyn_headband_straps_b_02, loc=<6 0 0>, rot=<0.1536 0.2648 -0.1738 -0.936>}
 dyn_headband_straps_b_04 {parent=dyn_headband_straps_b_03, loc=<6 0 0>, rot=<0.1864 0.1598 -0.0627 -0.9674>}
 clavicle_l {parent=spine_05, loc=<4.3575 -0.1466 -2.3223>, rot=<0.0465 -0.7289 -0.009 -0.6829>}
 upperarm_l {parent=clavicle_l, loc=<15.5111 0 0>, rot=<0.0226 -0.3234 -0.0088 -0.9459>}
 upperarm_twist_01_l {parent=upperarm_l, loc=<0.0041 0 0>, rot=<0 0 0 -1>}
 clavicle_r {parent=spine_05, loc=<4.3575 -0.1466 2.3223>, rot=<0.7289 0.0465 -0.6829 0.009>}
 upperarm_r {parent=clavicle_r, loc=<-15.5111 0 0>, rot=<0.0226 -0.3234 -0.0088 -0.9459>}
 upperarm_twist_01_r {parent=upperarm_r, loc=<-0.0041 0 0>, rot=<0 0 0 -1>}
 head_null {parent=spine_05, type=Null, loc=<20.6573 3.5489 0>, rot=<0.1219 0.0044 -0.0357 -0.9919>}
]
