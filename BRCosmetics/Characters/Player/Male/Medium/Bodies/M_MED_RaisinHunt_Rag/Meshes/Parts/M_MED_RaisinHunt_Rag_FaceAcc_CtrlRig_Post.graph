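# Post-process Control Rig for the RaisinHunt Rag character's face accessories, posing hair and headband strap bones over the animated skeleton.
# DeltaRotation functions and a spherical pose reader turn head rotation into weights for additive bone offsets.
# Pelvis height, crouch, aim-up, skydive, jump and parachute states gate further offsets.
# In the front end, outside emotes, backpack-belt state selects between strap offsets and weighted parent constraints to head, neck and torso.
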ControlRigBlueprint /BRCosmetics/Characters/Player/Male/Medium/Bodies/M_MED_RaisinHunt_Rag/Meshes/Parts/M_MED_RaisinHunt_Rag_FaceAcc_CtrlRig_Post
graphs [
 Multiply_4_ContainedGraph {nodes [
   n1 = RigVMFunctionEntryNode()
   RigVMFunctionReturnNode(Result=&RigVMFunction_MathDoubleMul(B=@n1.C, A=&RigVMFunction_MathDoubleMul(A=@n1.A, B=@n1.B).Result).Result)
  ]}
 Multiply_5_ContainedGraph {nodes [
   n1 = RigVMFunctionEntryNode()
   RigVMFunctionReturnNode(Result=&RigVMFunction_MathDoubleMul(B=@n1.C, A=&RigVMFunction_MathDoubleMul(A=@n1.A, B=@n1.B).Result).Result)
  ]}
 Multiply_5_ContainedGraph {nodes [
   n1 = RigVMFunctionEntryNode()
   RigVMFunctionReturnNode(Result=&RigVMFunction_MathDoubleMul(B=@n1.C, A=&RigVMFunction_MathDoubleMul(A=@n1.A, B=@n1.B).Result).Result)
  ]}
 New Function_ContainedGraph {nodes [
   n1 = RigVMFunctionEntryNode()
   RigVMFunction_MathBoolToInteger = &RigVMFunction_MathBoolToInteger(Value=@n1.onlyTwist)
   RigVMFunction_MathQuaternionSwingTwist = &RigVMFunction_MathQuaternionSwingTwist(
    twistAxis=@n1.twistAxis
    Input=&RigUnit_GetRelativeTransformForItem(bChildInitial=False, bParentInitial=False, child=@n1.child, parent=@n1.parent).RelativeTransform.Rotation
   )
   n2 = RigVMDispatchNode(
    Values=((),())
    Index=@RigVMFunction_MathBoolToInteger.Result
    Values.0=@RigVMFunction_MathQuaternionSwingTwist.Swing
    Values.1=@RigVMFunction_MathQuaternionSwingTwist.Twist
   ) {template="DISPATCH_RigVMDispatch_SelectInt32(in Index,in Values,out Result)"}
   RigVMFunction_MathQuaternionToEuler = &RigVMFunction_MathQuaternionToEuler(Value=@n2.Result, RotationOrder=@n1.Rotation Order)
   > DISPATCH_RigVMDispatch_Print = RigVMDispatchNode(ScreenColor=(R=1,G=1,B=1,A=1), ScreenDuration=1.5, Enabled=@n1.debugRotation, Value=@RigVMFunction_MathQuaternionToEuler.Result) {template="DISPATCH_RigVMDispatch_Print(in Prefix,in Value,in Enabled,in ScreenDuration,in ScreenColor)"}
   Cast = RigVMDispatchNode(Value=@n1.triggerAxis) {template="DISPATCH_RigVMDispatch_CastEnumToInt(in Value,out Result)"}
   DISPATCH_RigVMDispatch_SelectInt32 = RigVMDispatchNode(
    Values=(0,0,0)
    Index=@Cast.Result
    Values.0=@RigVMFunction_MathQuaternionToEuler.Result.X
    Values.1=@RigVMFunction_MathQuaternionToEuler.Result.Y
    Values.2=@RigVMFunction_MathQuaternionToEuler.Result.Z
   ) {template="DISPATCH_RigVMDispatch_SelectInt32(in Index,in Values,out Result)"}
   Remap = &RigVMFunction_MathDoubleRemap(
    bClamp=False
    Value=@DISPATCH_RigVMDispatch_SelectInt32.Result
    TargetMinimum=@n1.outputMinMax.X
    TargetMaximum=@n1.outputMinMax.Y
    SourceMinimum=@n1.rotationMinMax.X
    SourceMaximum=@n1.rotationMinMax.Y
   )
   Absolute = &RigVMFunction_MathDoubleAbs(Value=@Remap.Result)
   Multiply = &RigVMFunction_MathDoubleMul(
    A=&RigVMFunction_MathDoubleClamp(Value=RigVMDispatchNode(True=@Absolute.Result, Condition=@n1.absoluteValue, False=@Remap.Result) {template="DISPATCH_RigVMDispatch_If(in Condition,in True,in False,out Result)"}.Result).Result
    B=@n1.premult
   )
   > RigVMFunctionReturnNode(outputQuat=@n2.Result, outputWeight=@Multiply.Result)
  ]}
 New Function_ContainedGraph {nodes [
   n1 = RigVMFunctionEntryNode()
   RigVMFunction_MathBoolToInteger = &RigVMFunction_MathBoolToInteger(Value=@n1.onlyTwist)
   RigVMFunction_MathQuaternionSwingTwist = &RigVMFunction_MathQuaternionSwingTwist(
    twistAxis=@n1.twistAxis
    Input=&RigUnit_GetRelativeTransformForItem(bChildInitial=False, bParentInitial=False, child=@n1.child, parent=@n1.parent).RelativeTransform.Rotation
   )
   n2 = RigVMDispatchNode(
    Values=((),())
    Index=@RigVMFunction_MathBoolToInteger.Result
    Values.0=@RigVMFunction_MathQuaternionSwingTwist.Swing
    Values.1=@RigVMFunction_MathQuaternionSwingTwist.Twist
   ) {template="DISPATCH_RigVMDispatch_SelectInt32(in Index,in Values,out Result)"}
   RigVMFunction_MathQuaternionToEuler = &RigVMFunction_MathQuaternionToEuler(Value=@n2.Result, RotationOrder=@n1.Rotation Order)
   > DISPATCH_RigVMDispatch_Print = RigVMDispatchNode(ScreenColor=(R=1,G=1,B=1,A=1), ScreenDuration=1.5, Enabled=@n1.debugRotation, Value=@RigVMFunction_MathQuaternionToEuler.Result) {template="DISPATCH_RigVMDispatch_Print(in Prefix,in Value,in Enabled,in ScreenDuration,in ScreenColor)"}
   Cast = RigVMDispatchNode(Value=@n1.triggerAxis) {template="DISPATCH_RigVMDispatch_CastEnumToInt(in Value,out Result)"}
   DISPATCH_RigVMDispatch_SelectInt32 = RigVMDispatchNode(
    Values=(0,0,0)
    Index=@Cast.Result
    Values.0=@RigVMFunction_MathQuaternionToEuler.Result.X
    Values.1=@RigVMFunction_MathQuaternionToEuler.Result.Y
    Values.2=@RigVMFunction_MathQuaternionToEuler.Result.Z
   ) {template="DISPATCH_RigVMDispatch_SelectInt32(in Index,in Values,out Result)"}
   Remap = &RigVMFunction_MathDoubleRemap(
    bClamp=False
    Value=@DISPATCH_RigVMDispatch_SelectInt32.Result
    TargetMinimum=@n1.outputMinMax.X
    TargetMaximum=@n1.outputMinMax.Y
    SourceMinimum=@n1.rotationMinMax.X
    SourceMaximum=@n1.rotationMinMax.Y
   )
   Absolute = &RigVMFunction_MathDoubleAbs(Value=@Remap.Result)
   Multiply = &RigVMFunction_MathDoubleMul(
    A=&RigVMFunction_MathDoubleClamp(Value=RigVMDispatchNode(True=@Absolute.Result, Condition=@n1.absoluteValue, False=@Remap.Result) {template="DISPATCH_RigVMDispatch_If(in Condition,in True,in False,out Result)"}.Result).Result
    B=@n1.premult
   )
   > RigVMFunctionReturnNode(outputQuat=@n2.Result, outputWeight=@Multiply.Result)
  ]}
 RigVMFunctionLibrary {nodes [
   DeltaRotation = RigVMCollapseNode(
    ExecuteContext=()
    Rotation Order=ZYX
    child=(Type=Bone,Name="None")
    outputMinMax=(X=0,Y=1)
    parent=(Type=Bone,Name="None")
    premult=1
    rotationMinMax=(X=0,Y=0)
    twistAxis=(X=1,Y=0,Z=0)
   )
   DeltaRotation v4 = RigVMCollapseNode(
    ExecuteContext=()
    Rotation Order=ZYX
    child=(Type=Bone,Name="None")
    outputMinMax=(X=0,Y=1)
    parent=(Type=Bone,Name="None")
    premult=1
    rotationMinMax=(X=0,Y=0)
    twistAxis=(X=1,Y=0,Z=0)
   )
  ]}
 RigVMModel {nodes [
   RigUnit_BeginExecution = &RigUnit_BeginExecution()
   &RigUnit_ModifyTransforms(
    ItemToModify=((Item=(Type=Bone,Name="dyn_right_hair_a_05"),Transform=(Rotation=(X=0,Y=-0.1736,Z=0,W=0.9848),Translation=(X=0,Y=0,Z=0),Scale3D=(X=1,Y=1,Z=1))))
    Weight=1
    WeightMinimum=0
    ExecutePin=@RigUnit_BeginExecution.ExecutePin
   )
   n1 = &RigVMFunction_AnimEvalRichCurve(
    Curve="(EditorCurveData=(Keys=((InterpMode=RCIM_Cubic,TangentMode=RCTM_Break,TangentWeightMode=RCTWM_WeightedNone,Time=0,Value=0,ArriveTangent=0,ArriveTangentWeight=0,LeaveTangent=0,LeaveTangentWeight=0),(InterpMode=RCIM_Cubic,TangentMode=RCTM_Break,TangentWeight ...(truncated)"
    SourceMinimum=0.5
    TargetMinimum=0
    Value=&RigVMFunction_MathDoubleRemap(
     SourceMaximum=30
     SourceMinimum=60
     TargetMaximum=0
     TargetMinimum=1
     bClamp=False
     Value=&RigUnit_GetRelativeTransformForItem(Child=(Type=Bone,Name="pelvis"), Parent=(Type=Bone,Name="root"), bChildInitial=False, bParentInitial=False).RelativeTransform.Translation.Z
    ).Result
   )
   > RigVMVariableNode(Value=@n1.Result) {variable=GroundMovement}
   > DeltaRotation_4 = RigVMFunctionReferenceNode(
    Rotation Order=ZYX
    child=(Type=Bone,Name="head")
    outputMinMax=(X=0,Y=1)
    parent=(Type=Null,Name="head_null")
    premult=1
    rotationMinMax=(X=5,Y=-40)
    triggerAxis=Z
    twistAxis=(X=1,Y=0,Z=0)
   )
   > &RigUnit_ModifyTransforms(
    ItemToModify="((Item=(Type=Bone,Name=\"dyn_right_hair_01\"),Transform=(Rotation=(X=0,Y=-0.0872,Z=0,W=0.9962),Translation=(X=0,Y=0,Z=0),Scale3D=(X=1,Y=1,Z=1))),(Item=(Type=Bone,Name=\"dyn_middle_hair_01\"),Transform=(Rotation=(X=0,Y=-0.1305,Z=0,W=0.9914),Translation=(X=0,Y=0 ...(truncated)"
    WeightMinimum=0
    Weight=@DeltaRotation_4.outputWeight
   )
   > DeltaRotation_8 = RigVMFunctionReferenceNode(
    Rotation Order=ZYX
    child=(Type=Bone,Name="head")
    onlyTwist=true
    outputMinMax=(X=0,Y=1)
    parent=(Type=Null,Name="head_null")
    premult=1
    rotationMinMax=(X=0,Y=-50)
    triggerAxis=X
    twistAxis=(X=1,Y=0,Z=0)
   )
   > &RigUnit_ModifyTransforms(
    ItemToModify="((Item=(Type=Bone,Name=\"dyn_middle_hair_01\"),Transform=(Rotation=(X=0,Y=-0.1736,Z=0,W=0.9848),Translation=(X=0,Y=0,Z=0),Scale3D=(X=1,Y=1,Z=1))),(Item=(Type=Bone,Name=\"dyn_left_hair_01\"),Transform=(Rotation=(X=0,Y=0,Z=0,W=1),Translation=(X=0,Y=0,Z=0),Scale3 ...(truncated)"
    WeightMinimum=0
    Weight=@DeltaRotation_8.outputWeight
   )
   > n2 = &RigUnit_SphericalPoseReader(
    ActiveRegionScaleFactors=(PositiveWidth=1,NegativeWidth=1,PositiveHeight=1,NegativeHeight=1)
    ActiveRegionSize=0.05
    Debug=(bDrawDebug=false,bDraw2D=False,bDrawLocalAxes=False,DebugScale=25,DebugSegments=20,DebugThickness=0.25)
    DriverAxis=(X=1,Y=0,Z=0)
    DriverItem=(Type=Bone,Name="head")
    FalloffRegionScaleFactors=(PositiveWidth=1,NegativeWidth=1,PositiveHeight=1,NegativeHeight=1)
    FalloffSize=0.4
    FlipHeightScaling=False
    FlipWidthScaling=False
    OptionalParentItem=(Type=Bone,Name="spine_05")
    RotationOffset=(X=0,Y=-65,Z=-52)
   )
   RigVMFunction_MathBoolToFloat = &RigVMFunction_MathBoolToFloat(Value=RigVMVariableNode() {variable=Targeting AIm Up}.Value)
   n3 = &RigVMFunction_AlphaInterp(
    InterpSpeedDecreasing=10
    InterpSpeedIncreasing=10
    Scale=1
    bClampResult=False
    Value=&RigVMFunction_MathDoubleMul(B=@RigVMFunction_MathBoolToFloat.Result, A=@n2.OutputParam).Result
   )
   > &RigUnit_ModifyTransforms(
    ItemToModify="((Item=(Type=Bone,Name=\"dyn_right_hair_01\"),Transform=(Rotation=(X=0.0076,Y=-0.0868,Z=0.0868,W=0.9924),Translation=(X=0,Y=0,Z=0),Scale3D=(X=1,Y=1,Z=1))),(Item=(Type=Bone,Name=\"dyn_middle_hair_01\"),Transform=(Rotation=(X=0,Y=0,Z=0,W=1),Translation=(X=0,Y=0, ...(truncated)"
    WeightMinimum=0
    Weight=@n3.Result
   )
   > DeltaRotation_5 = RigVMFunctionReferenceNode(
    Rotation Order=ZYX
    child=(Type=Bone,Name="head")
    outputMinMax=(X=0,Y=1)
    parent=(Type=Null,Name="head_null")
    premult=1
    rotationMinMax=(X=-10,Y=-60)
    triggerAxis=Z
    twistAxis=(X=1,Y=0,Z=0)
   )
   Multiply_3 = RigVMAggregateNode(
    C=@DeltaRotation_5.outputWeight
    B=RigVMDispatchNode(False=0, True=1, Condition=RigVMVariableNode() {variable=Crouch}.Value) {template="DISPATCH_RigVMDispatch_If(in Condition,in True,in False,out Result)"}.Result
    A=RigVMVariableNode() {variable=GroundMovement}.Value
   )
   > &RigUnit_ModifyTransforms(
    ItemToModify="((Item=(Type=Bone,Name=\"dyn_left_hair_02\"),Transform=(Rotation=(X=0,Y=-0.1305,Z=0,W=0.9914),Translation=(X=0,Y=0,Z=0),Scale3D=(X=1,Y=1,Z=1))),(Item=(Type=Bone,Name=\"dyn_left_hair_a_01\"),Transform=(Rotation=(X=0,Y=-0.1305,Z=0,W=0.9914),Translation=(X=0,Y=0, ...(truncated)"
    WeightMinimum=0
    Weight=@Multiply_3.Result
   )
   n4 = &RigVMFunction_AlphaInterp(
    InterpSpeedDecreasing=10
    InterpSpeedIncreasing=10
    Scale=1
    bClampResult=False
    bInterpResult=False
    Value=&RigVMFunction_MathBoolToFloat(Value=RigVMVariableNode() {variable=Skydiving}.Value).Result
   )
   > &RigUnit_ModifyTransforms(
    ItemToModify="((Item=(Type=Bone,Name=\"dyn_left_hair_01\"),Transform=(Rotation=(X=0,Y=0.145,Z=0,W=-0.9894),Translation=(X=0,Y=0,Z=0),Scale3D=(X=1,Y=1,Z=1))),(Item=(Type=Bone,Name=\"dyn_middle_hair_01\"),Transform=(Rotation=(X=0.002,Y=0.0468,Z=-0.0436,W=-0.998),Translation=( ...(truncated)"
    WeightMinimum=0
    Weight=@n4.Result
   )
   AlphaInterp = &RigVMFunction_AlphaInterp(InterpSpeedDecreasing=3, InterpSpeedIncreasing=1, Scale=1, bClampResult=true, Value=RigVMVariableNode() {variable=IsJumping}.Value)
   > &RigUnit_ModifyTransforms(
    ItemToModify="((Item=(Type=Bone,Name=\"dyn_left_hair_02\"),Transform=(Rotation=(X=-0.0691,Y=-0.2601,Z=-0.2474,W=0.9308),Translation=(X=0,Y=0,Z=0),Scale3D=(X=1,Y=1,Z=1))),(Item=(Type=Bone,Name=\"dyn_middle_hair_02\"),Transform=(Rotation=(X=0.0749,Y=-0.2058,Z=0.3338,W=0.9168) ...(truncated)"
    WeightMinimum=0
    Weight=@AlphaInterp.Result
   )
   > &RigUnit_ModifyTransforms(
    ItemToModify="((Item=(Type=Bone,Name=\"dyn_left_hair_03\"),Transform=(Rotation=(X=0,Y=-0.1305,Z=0,W=0.9914),Translation=(X=0,Y=0,Z=0),Scale3D=(X=1,Y=1,Z=1))),(Item=(Type=Bone,Name=\"dyn_middle_hair_03\"),Transform=(Rotation=(X=0,Y=-0.1305,Z=0,W=0.9914),Translation=(X=0,Y=0, ...(truncated)"
    WeightMinimum=0
    Weight=@AlphaInterp.Result
   )
   > &RigUnit_ModifyTransforms(
    ItemToModify="((Item=(Type=Bone,Name=\"None\"),Transform=(Rotation=(X=0,Y=0,Z=0,W=1),Translation=(X=0,Y=0,Z=0),Scale3D=(X=1,Y=1,Z=1))),(Item=(Type=Bone,Name=\"dyn_left_hair_01\"),Transform=(Rotation=(X=0,Y=-0.1305,Z=0,W=0.9914),Translation=(X=0,Y=0,Z=0),Scale3D=(X=1,Y=1,Z=1 ...(truncated)"
    WeightMinimum=0
    Weight=@AlphaInterp.Result
   )
   > &RigUnit_ModifyTransforms(
    ItemToModify="((Item=(Type=Bone,Name=\"dyn_headband_straps_a_02\"),Transform=(Rotation=(X=0,Y=0,Z=-0.0175,W=0.9998),Translation=(X=0,Y=0,Z=0),Scale3D=(X=1,Y=1,Z=1))),(Item=(Type=Bone,Name=\"dyn_headband_straps_a_03\"),Transform=(Rotation=(X=0,Y=0,Z=0.0175,W=0.9998),Translat ...(truncated)"
    Weight=1
    WeightMinimum=0
   )
   > &RigUnit_ModifyTransforms(
    ItemToModify="((Item=(Type=Bone,Name=\"dyn_headband_straps_a_02\"),Transform=(Rotation=(X=0,Y=0,Z=-0.0436,W=0.999),Translation=(X=0,Y=0,Z=0),Scale3D=(X=1,Y=1,Z=1))),(Item=(Type=Bone,Name=\"dyn_headband_straps_a_03\"),Transform=(Rotation=(X=0,Y=0,Z=0.0175,W=0.9998),Translati ...(truncated)"
    Weight=1
    WeightMinimum=0
   )
   Multiply = &RigVMFunction_MathDoubleMul(
    B=RigVMDispatchNode(False=1, True=0, Condition=RigVMVariableNode() {variable=BackPackBelt}.Value) {template="DISPATCH_RigVMDispatch_If(in Condition,in True,in False,out Result)"}.Result
    A=&RigVMFunction_MathBoolToFloat(Value=RigVMVariableNode() {variable=ParachuteOpened}.Value).Result
   )
   > &RigUnit_ModifyTransforms(
    ItemToModify=((Item=(Type=Bone,Name="dyn_headband_straps_b_02"),Transform=(Rotation=(X=0.1574,Y=0.4162,Z=-0.0734,W=0.8925),Translation=(X=0,Y=0,Z=0),Scale3D=(X=1,Y=1,Z=1))))
    WeightMinimum=0
    Weight=@Multiply.Result
   )
   > &RigUnit_OffsetTransformForItem(
    Item=(Type=Bone,Name="dyn_headband_straps_b_02")
    OffsetTransform=(Rotation=(X=0,Y=0,Z=0,W=1),Translation=(X=0,Y=0,Z=5),Scale3D=(X=1,Y=1,Z=1))
    Weight=1
    bPropagateToChildren=false
   )
   > &RigUnit_OffsetTransformForItem(
    Item=(Type=Bone,Name="dyn_headband_straps_b_03")
    OffsetTransform=(Rotation=(X=0,Y=0,Z=0,W=1),Translation=(X=0,Y=0,Z=3),Scale3D=(X=1,Y=1,Z=1))
    Weight=1
    bPropagateToChildren=false
   )
   > DeltaRotation_2 = RigVMFunctionReferenceNode(
    Rotation Order=ZYX
    child=(Type=Bone,Name="head")
    outputMinMax=(X=0,Y=1)
    parent=(Type=Null,Name="head_null")
    premult=1
    rotationMinMax=(X=0,Y=50)
    triggerAxis=Z
    twistAxis=(X=1,Y=0,Z=0)
   )
   > &RigUnit_ModifyTransforms(
    ItemToModify=((Item=(Type=Bone,Name="dyn_headband_straps_a_02"),Transform=(Rotation=(X=0,Y=0,Z=0.0872,W=0.9962),Translation=(X=0,Y=0,Z=0),Scale3D=(X=1,Y=1,Z=1))))
    WeightMinimum=0
    Weight=@DeltaRotation_2.outputWeight
   )
   Multiply_1 = RigVMAggregateNode(
    B=RigVMDispatchNode(False=1, True=0, Condition=RigVMVariableNode() {variable=PlayingEmote}.Value) {template="DISPATCH_RigVMDispatch_If(in Condition,in True,in False,out Result)"}.Result
    C=RigVMDispatchNode(False=0, True=1, Condition=RigVMVariableNode() {variable=BackPackBelt}.Value) {template="DISPATCH_RigVMDispatch_If(in Condition,in True,in False,out Result)"}.Result
    A=&RigVMFunction_MathBoolToFloat(Value=RigVMVariableNode() {variable=FrontEnd}.Value).Result
   )
   n5 = &RigVMFunction_AlphaInterp(InterpSpeedDecreasing=10, InterpSpeedIncreasing=10, Scale=1, bClampResult=False, bInterpResult=False, Value=@Multiply_1.Result)
   > &RigUnit_ModifyTransforms(
    ItemToModify="((Item=(Type=Bone,Name=\"dyn_headband_straps_a_03\"),Transform=(Rotation=(X=0.0019,Y=0.0436,Z=-0.0436,W=0.9981),Translation=(X=0,Y=0,Z=0),Scale3D=(X=1,Y=1,Z=1))),(Item=(Type=Bone,Name=\"dyn_headband_straps_a_04\"),Transform=(Rotation=(X=0.0057,Y=-0.0432,Z=0.13 ...(truncated)"
    WeightMinimum=0
    Weight=@n5.Result
   )
   Multiply_4 = RigVMAggregateNode(
    B=RigVMDispatchNode(False=1, True=0, Condition=RigVMVariableNode() {variable=PlayingEmote}.Value) {template="DISPATCH_RigVMDispatch_If(in Condition,in True,in False,out Result)"}.Result
    C=RigVMDispatchNode(False=1, True=0, Condition=RigVMVariableNode() {variable=BackPackBelt}.Value) {template="DISPATCH_RigVMDispatch_If(in Condition,in True,in False,out Result)"}.Result
    A=&RigVMFunction_MathBoolToFloat(Value=RigVMVariableNode() {variable=FrontEnd}.Value).Result
   )
   n6 = &RigVMFunction_AlphaInterp(InterpSpeedDecreasing=10, InterpSpeedIncreasing=10, Scale=1, bClampResult=False, bInterpResult=False, Value=@Multiply_4.Result)
   > &RigUnit_ParentConstraint(
    AdvancedSettings=(RotationOrderForFilter=XZY)
    Child=(Type=Bone,Name="dyn_headband_straps_a_01")
    Parents=((Item=(Type=Bone,Name="head"),Weight=1.0))
    Weight=@n6.Result
   )
   > &RigUnit_ParentConstraint(
    AdvancedSettings=(RotationOrderForFilter=XZY)
    Child=(Type=Bone,Name="dyn_headband_straps_a_02")
    Parents=((Item=(Type=Bone,Name="neck_02"),Weight=0.03),(Item=(Type=Bone,Name="head"),Weight=0.97))
    Weight=@n6.Result
   )
   > &RigUnit_ParentConstraint(
    AdvancedSettings=(RotationOrderForFilter=XZY)
    Child=(Type=Bone,Name="dyn_headband_straps_a_03")
    Parents="((Item=(Type=Bone,Name=\"spine_03\"),Weight=0.0003),(Item=(Type=Bone,Name=\"deform_pec_l\"),Weight=0.0004),(Item=(Type=Bone,Name=\"spine_04\"),Weight=0.0012),(Item=(Type=Bone,Name=\"clavicle_l\"),Weight=0.017),(Item=(Type=Bone,Name=\"neck_01\"),Weight=0.0463),(Item= ...(truncated)"
    Weight=@n6.Result
   )
   > &RigUnit_ParentConstraint(
    AdvancedSettings=(RotationOrderForFilter=XZY)
    Child=(Type=Bone,Name="dyn_headband_straps_a_04")
    Parents="((Item=(Type=Bone,Name=\"spine_03\"),Weight=0.0103),(Item=(Type=Bone,Name=\"head\"),Weight=0.0124),(Item=(Type=Bone,Name=\"deform_pec_l\"),Weight=0.0146),(Item=(Type=Bone,Name=\"neck_01\"),Weight=0.0147),(Item=(Type=Bone,Name=\"neck_02\"),Weight=0.0198),(Item=(Type= ...(truncated)"
    Weight=@n6.Result
   )
   > &RigUnit_ParentConstraint(
    AdvancedSettings=(RotationOrderForFilter=XZY)
    Child=(Type=Bone,Name="dyn_headband_straps_a_05")
    Parents=((Item=(Type=Bone,Name="spine_03"),Weight=0.066),(Item=(Type=Bone,Name="deform_pec_l"),Weight=0.1538),(Item=(Type=Bone,Name="spine_04"),Weight=0.1734),(Item=(Type=Bone,Name="clavicle_l"),Weight=0.2653),(Item=(Type=Bone,Name="spine_05"),Weight=0.3415))
    Weight=@n6.Result
   )
   > &RigUnit_OffsetTransformForItem(
    Item=(Type=Bone,Name="dyn_headband_straps_a_02")
    OffsetTransform=(Rotation=(X=0,Y=0,Z=0.0436,W=0.999),Translation=(X=0,Y=1,Z=2),Scale3D=(X=1,Y=1,Z=1))
    bPropagateToChildren=true
    Weight=@n6.Result
   )
   > &RigUnit_OffsetTransformForItem(
    Item=(Type=Bone,Name="dyn_headband_straps_a_03")
    OffsetTransform=(Rotation=(X=0,Y=0,Z=-0.0872,W=0.9962),Translation=(X=0,Y=1,Z=1),Scale3D=(X=1,Y=1,Z=1))
    bPropagateToChildren=true
    Weight=@n6.Result
   )
   > &RigUnit_OffsetTransformForItem(
    Item=(Type=Bone,Name="dyn_headband_straps_a_05")
    OffsetTransform=(Rotation=(X=0,Y=0,Z=0.1736,W=0.9848),Translation=(X=0,Y=0,Z=0),Scale3D=(X=1,Y=1,Z=1))
    bPropagateToChildren=true
    Weight=@n6.Result
   )
  ]}
]
hierarchy [
 root {loc=<0 0 0>, rot=<0 0 0 1>}
 pelvis {parent=root, loc=<0 -0.7071 93.1191>, rot=<0.0127 0.707 -0.0127 -0.707>}
 spine_01 {parent=pelvis, loc=<6.7494 -0.6388 0>, rot=<0 0 0.034 -0.9994>}
 spine_02 {parent=spine_01, loc=<7.2518 -0.035 0>, rot=<0 0 0.0238 -0.9997>}
 spine_03 {parent=spine_02, loc=<6.0104 0 0>, rot=<0 0 0.0149 -0.9999>}
 spine_04 {parent=spine_03, loc=<8.1024 0 0>, rot=<0 0 -0.0132 -0.9999>}
 spine_05 {parent=spine_04, loc=<10.0146 0 0>, rot=<0 0 -0.0015 -1>}
 neck_01 {parent=spine_05, loc=<11.3138 0.1529 0>, rot=<0 0 -0.2003 -0.9797>}
 neck_02 {parent=neck_01, loc=<4.5627 0 0>, rot=<0 0 0.0504 -0.9987>}
 head {parent=neck_02, loc=<5.3914 0 0>, rot=<0 0 0.115 -0.9934>}
 dyn_headband_straps_a_base {parent=head, loc=<7.0799 1.4354 -6.8846>, rot=<0.0411 -0.9991 -0.0004 -0.0102>}
 dyn_headband_straps_a_01 {parent=dyn_headband_straps_a_base, loc=<0 0 0>, rot=<0 0 0 1>}
 dyn_headband_straps_a_02 {parent=dyn_headband_straps_a_01, loc=<6 0 0>, rot=<0.0007 0.0217 -0.0304 -0.9993>}
 dyn_headband_straps_a_03 {parent=dyn_headband_straps_a_02, loc=<6 0 0>, rot=<0.1993 -0.0037 -0.2715 -0.9416>}
 dyn_headband_straps_a_04 {parent=dyn_headband_straps_a_03, loc=<6 0 0>, rot=<0.2434 0.0652 0.0036 -0.9677>}
 dyn_headband_straps_a_05 {parent=dyn_headband_straps_a_04, loc=<6 0 0>, rot=<0.1442 0 0 -0.9895>}
 dyn_right_hair_c_base {parent=head, loc=<6.781 1.9354 7.702>, rot=<0 0 1 0.009>}
 dyn_right_hair_c_01 {parent=dyn_right_hair_c_base, loc=<0 0 0>, rot=<0 0 0 1>}
 dyn_front_r_hair_base {parent=head, loc=<10.6966 9.0808 3.4445>, rot=<0.0064 0.7071 0.7071 0.0064>}
 dyn_front_r_hair_01 {parent=dyn_front_r_hair_base, loc=<0 0 0>, rot=<0 0 0 1>}
 dyn_front_l_hair_base {parent=head, loc=<10.6021 9.2978 -0.7976>, rot=<0.0064 0.7071 0.7071 0.0064>}
 dyn_front_l_hair_01 {parent=dyn_front_l_hair_base, loc=<0 0 0>, rot=<0 0 0 1>}
 dyn_left_hair_c_base {parent=head, loc=<9.6876 4.4139 -6.7563>, rot=<-0.009 -1 0 0>}
 dyn_left_hair_c_01 {parent=dyn_left_hair_c_base, loc=<0 0 0>, rot=<0 0 0 1>}
 dyn_back_hair_a_base {parent=head, loc=<9.1651 -8.1458 -3.626>, rot=<-0.0068 -0.7515 0.6596 0.006>}
 dyn_back_hair_a_01 {parent=dyn_back_hair_a_base, loc=<0 0 0>, rot=<0 0 0 1>}
 dyn_back_hair_b_base {parent=head, loc=<9.1957 -8.4865 1.4467>, rot=<0.0317 -0.7064 0.7064 -0.0317>}
 dyn_back_hair_b_01 {parent=dyn_back_hair_b_base, loc=<0 0 0>, rot=<0 0 0 1>}
 dyn_left_hair_b_base {parent=head, loc=<9.1216 -5.7876 -7.1321>, rot=<0.0168 -0.9616 0.2695 -0.0495>}
 dyn_left_hair_b_01 {parent=dyn_left_hair_b_base, loc=<0 0 0>, rot=<0 0 0 1>}
 dyn_middle_hair_base {parent=head, loc=<5.1887 -7.3721 0>, rot=<0.1132 -0.698 0.698 -0.1132>}
 dyn_middle_hair_01 {parent=dyn_middle_hair_base, loc=<0 0 0>, rot=<0 0 0 1>}
 dyn_middle_hair_02 {parent=dyn_middle_hair_01, loc=<6 0 0>, rot=<0 0 0 1>}
 dyn_middle_hair_03 {parent=dyn_middle_hair_02, loc=<6 0 0>, rot=<0 0 0 1>}
 dyn_middle_hair_04 {parent=dyn_middle_hair_03, loc=<6 0 0>, rot=<0 0 0 1>}
 dyn_right_hair_base {parent=head, loc=<5.0257 -6.9089 5.3399>, rot=<0.1149 -0.5761 0.8012 -0.1143>}
 dyn_right_hair_01 {parent=dyn_right_hair_base, loc=<0 0 0>, rot=<0 0 0 1>}
 dyn_right_hair_02 {parent=dyn_right_hair_01, loc=<6 0 0>, rot=<0 0 0 1>}
 dyn_right_hair_03 {parent=dyn_right_hair_02, loc=<6 0 0>, rot=<0 0 0 1>}
 dyn_right_hair_04 {parent=dyn_right_hair_03, loc=<6 0 0>, rot=<0 0 0 1>}
 dyn_right_hair_a_base {parent=head, loc=<5.9634 -3.4254 7.7828>, rot=<0.0322 -0.1771 0.9798 -0.0875>}
 dyn_right_hair_a_01 {parent=dyn_right_hair_a_base, loc=<0 0 0>, rot=<0 0 0 1>}
 dyn_right_hair_a_02 {parent=dyn_right_hair_a_01, loc=<6 0 0>, rot=<0 0 0 1>}
 dyn_right_hair_a_03 {parent=dyn_right_hair_a_02, loc=<6 0 0>, rot=<0 0 0 1>}
 dyn_right_hair_a_04 {parent=dyn_right_hair_a_03, loc=<6 0 0>, rot=<0 0 0 1>}
 dyn_right_hair_a_05 {parent=dyn_right_hair_a_02, loc=<1.4226 -0.8845 -1.04>, rot=<-0.1975 -0.0807 0.3247 -0.9214>}
 dyn_left_hair_a_base {parent=head, loc=<5.6938 -3.0382 -7.9307>, rot=<0.0991 -0.9834 0.1494 -0.0269>}
 dyn_left_hair_a_01 {parent=dyn_left_hair_a_base, loc=<0 0 0>, rot=<0 0 0 1>}
 dyn_left_hair_a_02 {parent=dyn_left_hair_a_01, loc=<6 0 0>, rot=<0 0 0 1>}
 dyn_left_hair_a_03 {parent=dyn_left_hair_a_02, loc=<6 0 0>, rot=<0 0 0 1>}
 dyn_left_hair_a_04 {parent=dyn_left_hair_a_03, loc=<6 0 0>, rot=<0 0 0 1>}
 dyn_left_hair_base {parent=head, loc=<5.0117 -6.8292 -5.1299>, rot=<0.1201 -0.8696 0.4667 -0.1071>}
 dyn_left_hair_01 {parent=dyn_left_hair_base, loc=<0 0 0>, rot=<0 0 0 1>}
 dyn_left_hair_02 {parent=dyn_left_hair_01, loc=<6 0 0>, rot=<0 0 0 1>}
 dyn_left_hair_03 {parent=dyn_left_hair_02, loc=<6 0 0>, rot=<0 0 0 1>}
 dyn_left_hair_04 {parent=dyn_left_hair_03, loc=<6 0 0>, rot=<0 0 0 1>}
 dyn_right_hair_b_base {parent=head, loc=<9.2987 -3.6621 7.8853>, rot=<-0.0048 -0.2179 0.9658 -0.1409>}
 dyn_right_hair_b_01 {parent=dyn_right_hair_b_base, loc=<0 0 0>, rot=<0 0 0 1>}
 dyn_headband_straps_b_base {parent=head, loc=<7.091 1.3015 -6.8408>, rot=<0.0411 -0.9989 0.0009 0.0225>}
 dyn_headband_straps_b_01 {parent=dyn_headband_straps_b_base, loc=<0 0 0>, rot=<0 0 0 1>}
 dyn_headband_straps_b_02 {parent=dyn_headband_straps_b_01, loc=<6 0 0>, rot=<0.0352 -0.0211 0.1028 -0.9939>}
 dyn_headband_straps_b_03 {parent=dyn_headband_straps_b_02, loc=<6 0 0>, rot=<0.1536 0.2648 -0.1738 -0.936>}
 dyn_headband_straps_b_04 {parent=dyn_headband_straps_b_03, loc=<6 0 0>, rot=<0.1864 0.1598 -0.0627 -0.9674>}
 clavicle_l {parent=spine_05, loc=<4.3575 -0.1466 -2.3223>, rot=<0.0465 -0.7289 -0.009 -0.6829>}
 upperarm_l {parent=clavicle_l, loc=<15.5111 0 0>, rot=<0.0226 -0.3234 -0.0088 -0.9459>}
 upperarm_twist_01_l {parent=upperarm_l, loc=<0.0041 0 0>, rot=<0 0 0 -1>}
 deform_upperarm_bk_base_l {parent=upperarm_twist_01_l, loc=<1.0813 -7.2193 -1.9453>, rot=<0 0 0 1>}
 deform_upperarm_bk_l {parent=deform_upperarm_bk_base_l, loc=<0 0 0>, rot=<0 0 0 1>}
 deform_upperarm_fr_base_l {parent=upperarm_twist_01_l, loc=<-0.3691 7.0506 -2.765>, rot=<0 0 0 1>}
 deform_upperarm_fr_l {parent=deform_upperarm_fr_base_l, loc=<0 0 0>, rot=<0 0 0 1>}
 deform_upperarm_up_base_l {parent=upperarm_twist_01_l, loc=<-3.2216 -1.1391 3.3311>, rot=<0 0 0 1>}
 deform_upperarm_up_l {parent=deform_upperarm_up_base_l, loc=<0 0 0>, rot=<0 0 0 1>}
 lowerarm_l {parent=upperarm_l, loc=<26.6514 0 0>, rot=<0.0391 -0.0099 -0.2453 -0.9686>}
 hand_l {parent=lowerarm_l, loc=<22.3491 0 0>, rot=<-0.7081 -0.0507 0.0975 -0.6975>}
 thumb_01_l {parent=hand_l, loc=<1.3118 -0.3191 -2.459>, rot=<0.6546 -0.4042 0.0664 -0.6354>}
 thumb_02_l {parent=thumb_01_l, loc=<5.9336 -0.0448 0.0401>, rot=<-0.0037 0.0213 0.298 -0.9543>}
 thumb_03_l {parent=thumb_02_l, loc=<2.2223 0.3682 -0.0907>, rot=<0.0118 -0.013 0.0066 -0.9998>}
 index_metacarpal_l {parent=hand_l, loc=<1.6535 0.0183 -2.96>, rot=<0.0158 -0.0124 -0.0173 -0.9996>}
 index_01_l {parent=index_metacarpal_l, loc=<8.3008 0.0357 -0.0524>, rot=<-0.0044 0.0485 0.0924 -0.9945>}
 index_02_l {parent=index_01_l, loc=<3.3587 -0.3004 0.0452>, rot=<-0.002 0.0027 0.3225 -0.9466>}
 index_03_l {parent=index_02_l, loc=<2.5578 -0.123 -0.0252>, rot=<-0.002 0.0067 0.2558 -0.9667>}
 middle_metacarpal_l {parent=hand_l, loc=<2.1856 -0.1725 -0.5984>, rot=<0.006 0.0168 -0.0329 -0.9993>}
 middle_01_l {parent=middle_metacarpal_l, loc=<7.3275 -0.2447 0.0129>, rot=<-0.0208 0.0042 0.0858 -0.9961>}
 middle_02_l {parent=middle_01_l, loc=<4.0123 0.1401 0.1175>, rot=<-0.0218 0.0258 0.3512 -0.9357>}
 middle_03_l {parent=middle_02_l, loc=<2.6077 0.0534 0.1373>, rot=<0.0082 0.0188 0.1708 -0.9851>}
 ring_metacarpal_l {parent=hand_l, loc=<2.2074 -0.3625 1.3638>, rot=<-0.0284 0.0722 -0.0458 -0.9959>}
 ring_01_l {parent=ring_metacarpal_l, loc=<7.0295 -0.6197 -0.1446>, rot=<-0.0489 -0.0252 0.1892 -0.9804>}
 ring_02_l {parent=ring_01_l, loc=<3.6174 0.4647 -0.1027>, rot=<0 0 0.2806 -0.9598>}
 ring_03_l {parent=ring_02_l, loc=<2.4152 0.114 0.0199>, rot=<0 0 0.1189 -0.9929>}
 pinky_metacarpal_l {parent=hand_l, loc=<1.7616 -0.9753 3.3814>, rot=<-0.0397 0.0814 -0.069 -0.9935>}
 pinky_01_l {parent=pinky_metacarpal_l, loc=<7.1709 -0.4831 -0.1094>, rot=<-0.0654 -0.0035 0.2242 -0.9723>}
 pinky_02_l {parent=pinky_01_l, loc=<2.8274 0.0893 -0.1043>, rot=<0.0037 0.0011 0.2854 -0.9584>}
 pinky_03_l {parent=pinky_02_l, loc=<2.2046 0.0307 -0.0078>, rot=<-0.0002 0.0098 0.1445 -0.9895>}
 deform_wrist_dn_base_l {parent=hand_l, loc=<0 -2.2342 0>, rot=<0 0 0 -1>}
 deform_wrist_dn_l {parent=deform_wrist_dn_base_l, loc=<0 0 0>, rot=<0 0 0 1>}
 deform_wrist_up_base_l {parent=hand_l, loc=<0 2.2682 0>, rot=<0 0 0 -1>}
 deform_wrist_up_l {parent=deform_wrist_up_base_l, loc=<0 0 0>, rot=<0 0 0 1>}
 weapon_l {parent=hand_l, loc=<-0.1974 -3.4111 0.3811>, rot=<0 0 0 1>}
 hand_attach_l {parent=hand_l, loc=<8.5 -3 0>, rot=<0.7071 0.7071 0 0>}
 lowerarm_twist_02_l {parent=lowerarm_l, loc=<9.6199 0 0>, rot=<0 0 0 -1>}
 lowerarm_twist_01_l {parent=lowerarm_l, loc=<17.8549 0 0>, rot=<0 0 0 -1>}
 deform_elbow_in_base_l {parent=lowerarm_l, loc=<0.2112 5.259 0.8274>, rot=<0 0 0 -1>}
 deform_elbow_in_l {parent=deform_elbow_in_base_l, loc=<0 0 0>, rot=<0 0 0 1>}
 deform_elbow_out_base_l {parent=lowerarm_l, loc=<0 -4 0>, rot=<0 0 0 -1>}
 deform_elbow_out_l {parent=deform_elbow_out_base_l, loc=<0 0 0>, rot=<0 0 0 1>}
 upperarm_twist_02_l {parent=upperarm_l, loc=<13.7941 0 0>, rot=<0 0 0 -1>}
 deform_bicep_base_master {parent=upperarm_twist_02_l, loc=<0 0 0>, rot=<0 0 0 1>}
 base_upperarm_tricep_l_master {parent=upperarm_twist_02_l, loc=<0 0 0>, rot=<0 0 0 1>}
 upperarm_tricep_l_base {parent=upperarm_twist_02_l, loc=<-1.7068 -3.9968 0.0002>, rot=<0 0 -0.7232 -0.6907>}
 upperarm_tricep_l {parent=upperarm_tricep_l_base, loc=<0 0 0>, rot=<0 0 0 1>}
 upperarm_bicep_l_base {parent=upperarm_twist_02_l, loc=<1.1167 4.4381 0.0002>, rot=<0 0 -0.695 -0.719>}
 upperarm_bicep_l {parent=upperarm_bicep_l_base, loc=<0 0 0>, rot=<0 0 0 1>}
 upperarm_l_null {parent=clavicle_l, type=Null, loc=<15.9777 12.7258 -4.4727>, rot=<0.2211 -0.4442 -0.7301 -0.4699>}
 clavicle_r {parent=spine_05, loc=<4.3575 -0.1466 2.3223>, rot=<0.7289 0.0465 -0.6829 0.009>}
 upperarm_r {parent=clavicle_r, loc=<-15.5111 0 0>, rot=<0.0226 -0.3234 -0.0088 -0.9459>}
 upperarm_twist_01_r {parent=upperarm_r, loc=<-0.0041 0 0>, rot=<0 0 0 -1>}
 deform_upperarm_bk_base_r {parent=upperarm_twist_01_r, loc=<-1.0814 7.2192 1.9455>, rot=<0 0 0 1>}
 deform_upperarm_bk_r {parent=deform_upperarm_bk_base_r, loc=<0 0 0>, rot=<0 0 0 1>}
 deform_upperarm_fr_base_r {parent=upperarm_twist_01_r, loc=<0.3689 -7.0506 2.7652>, rot=<0 0 0 1>}
 deform_upperarm_fr_r {parent=deform_upperarm_fr_base_r, loc=<0 0 0>, rot=<0 0 0 1>}
 deform_upperarm_up_base_r {parent=upperarm_twist_01_r, loc=<3.2214 1.1391 -3.3308>, rot=<0 0 0 1>}
 deform_upperarm_up_r {parent=deform_upperarm_up_base_r, loc=<0 0 0>, rot=<0 0 0 -1>}
 lowerarm_r {parent=upperarm_r, loc=<-26.6514 0 0>, rot=<0.0391 -0.0099 -0.2453 -0.9686>}
 hand_r {parent=lowerarm_r, loc=<-22.3491 0 0>, rot=<-0.7081 -0.0507 0.0975 -0.6975>}
 thumb_01_r {parent=hand_r, loc=<-1.3118 0.3191 2.459>, rot=<0.6546 -0.4042 0.0664 -0.6354>}
 thumb_02_r {parent=thumb_01_r, loc=<-5.9336 0.0448 -0.0401>, rot=<-0.0037 0.0213 0.298 -0.9543>}
 thumb_03_r {parent=thumb_02_r, loc=<-2.2223 -0.3682 0.0907>, rot=<0.0118 -0.013 0.0066 -0.9998>}
 index_metacarpal_r {parent=hand_r, loc=<-1.6535 -0.0183 2.96>, rot=<0.0158 -0.0124 -0.0173 -0.9996>}
 index_01_r {parent=index_metacarpal_r, loc=<-8.3008 -0.0357 0.0524>, rot=<-0.0044 0.0485 0.0924 -0.9945>}
 index_02_r {parent=index_01_r, loc=<-3.3587 0.3004 -0.0452>, rot=<-0.002 0.0027 0.3225 -0.9466>}
 index_03_r {parent=index_02_r, loc=<-2.5578 0.123 0.0252>, rot=<-0.002 0.0067 0.2558 -0.9667>}
 middle_metacarpal_r {parent=hand_r, loc=<-2.1856 0.1725 0.5984>, rot=<0.006 0.0168 -0.0329 -0.9993>}
 middle_01_r {parent=middle_metacarpal_r, loc=<-7.3275 0.2447 -0.0129>, rot=<-0.0208 0.0042 0.0858 -0.9961>}
 middle_02_r {parent=middle_01_r, loc=<-4.0123 -0.1401 -0.1175>, rot=<-0.0218 0.0258 0.3512 -0.9357>}
 middle_03_r {parent=middle_02_r, loc=<-2.6077 -0.0534 -0.1373>, rot=<0.0082 0.0188 0.1708 -0.9851>}
 ring_metacarpal_r {parent=hand_r, loc=<-2.2074 0.3625 -1.3638>, rot=<-0.0284 0.0722 -0.0458 -0.9959>}
 ring_01_r {parent=ring_metacarpal_r, loc=<-7.0295 0.6197 0.1446>, rot=<-0.0489 -0.0252 0.1892 -0.9804>}
 ring_02_r {parent=ring_01_r, loc=<-3.6174 -0.4647 0.1027>, rot=<0 0 0.2806 -0.9598>}
 ring_03_r {parent=ring_02_r, loc=<-2.4152 -0.114 -0.0199>, rot=<0 0 0.1189 -0.9929>}
 pinky_metacarpal_r {parent=hand_r, loc=<-1.7616 0.9753 -3.3814>, rot=<-0.0397 0.0814 -0.069 -0.9935>}
 pinky_01_r {parent=pinky_metacarpal_r, loc=<-7.1709 0.4831 0.1094>, rot=<-0.0654 -0.0035 0.2242 -0.9723>}
 pinky_02_r {parent=pinky_01_r, loc=<-2.8274 -0.0893 0.1043>, rot=<0.0037 0.0011 0.2854 -0.9584>}
 pinky_03_r {parent=pinky_02_r, loc=<-2.2046 -0.0307 0.0078>, rot=<-0.0002 0.0098 0.1445 -0.9895>}
 deform_wrist_dn_base_r {parent=hand_r, loc=<-0.0004 2.2345 -0.0001>, rot=<0 0 0 -1>}
 deform_wrist_dn_r {parent=deform_wrist_dn_base_r, loc=<0 0 0>, rot=<0 0 0 1>}
 deform_wrist_up_base_r {parent=hand_r, loc=<0.0003 -2.2684 0.0001>, rot=<0 0 0 -1>}
 deform_wrist_up_r {parent=deform_wrist_up_base_r, loc=<0 0 0>, rot=<0 0 0 1>}
 weapon_r {parent=hand_r, loc=<0.1974 3.4111 -0.3811>, rot=<0 0 0 1>}
 hand_attach_r {parent=hand_r, loc=<-8.5001 2.9999 0>, rot=<0 0 -0.7071 -0.7071>}
 lowerarm_twist_02_r {parent=lowerarm_r, loc=<-9.6199 0 0>, rot=<0 0 0 -1>}
 lowerarm_twist_01_r {parent=lowerarm_r, loc=<-17.8549 0 0>, rot=<0 0 0 -1>}
 deform_elbow_in_base_r {parent=lowerarm_r, loc=<-0.2111 -5.2591 -0.8275>, rot=<0 0 0 -1>}
 deform_elbow_in_r {parent=deform_elbow_in_base_r, loc=<-0.0001 0 0.0002>, rot=<0 0 0 1>}
 deform_elbow_out_base_r {parent=lowerarm_r, loc=<0 4 -0.0001>, rot=<0 0 0 -1>}
 deform_elbow_out_r {parent=deform_elbow_out_base_r, loc=<0 0 0>, rot=<0 0 0 1>}
 upperarm_twist_02_r {parent=upperarm_r, loc=<-13.7941 0 0>, rot=<0 0 0 -1>}
 upperarm_bicep_r_base {parent=upperarm_twist_02_r, loc=<-1.1168 -4.4381 -0.0001>, rot=<0 0 0.719 -0.695>}
 upperarm_bicep_r {parent=upperarm_bicep_r_base, loc=<0 0 0>, rot=<0 0 0 1>}
 upperarm_tricep_r_base {parent=upperarm_twist_02_r, loc=<1.7066 3.9968 -0.0001>, rot=<0 0 0.6907 -0.7232>}
 upperarm_tricep_r {parent=upperarm_tricep_r_base, loc=<0 0 0>, rot=<0 0 0 1>}
 dyn_top_clavicle_bullet_master {parent=clavicle_r, loc=<-9.958 0.4514 -8.0171>, rot=<0.0126 0.6113 -0.0554 -0.7893>}
 dyn_top_clavicle_bullet_01 {parent=dyn_top_clavicle_bullet_master, loc=<0 0 0>, rot=<0 0 0 1>}
 head_null {parent=spine_05, type=Null, loc=<20.6573 3.5489 0>, rot=<0.1219 0.0044 -0.0357 -0.9919>}
 clavicle_l_null {parent=spine_05, type=Null, loc=<4.3575 -0.1466 -2.3223>, rot=<0.0465 -0.7289 -0.009 -0.6829>}
 attach_cape {parent=spine_05, loc=<9.017 -11.714 0>, rot=<0.0616 -0.7044 -0.0616 -0.7044>}
 attach_backpack {parent=spine_05, loc=<-0.123 -13.473 0>, rot=<-0.0093 -0.707 0.0093 -0.707>}
 dyn_front_clavicle_bullet_master {parent=spine_05, loc=<5.8139 7.9477 9.8621>, rot=<-0.3023 -0.0872 0.9188 0.2385>}
 dyn_front_clavicle_bullet_01 {parent=dyn_front_clavicle_bullet_master, loc=<0 0 0>, rot=<0 0 0 1>}
 dyn_back_clavicle_bullet_master {parent=spine_05, loc=<7.6564 -10.4899 7.4895>, rot=<-0.0456 -0.7519 0.5701 -0.328>}
 dyn_back_clavicle_bullet_01 {parent=dyn_back_clavicle_bullet_master, loc=<0 0 0>, rot=<0 0 0 1>}
 dyn_necklace_base {parent=spine_05, loc=<8.4723 7.1942 0>, rot=<0.4795 0.5196 0.5196 0.4795>}
 dyn_necklace_01 {parent=dyn_necklace_base, loc=<0 0 0>, rot=<0 -0.5413 0 -0.8408>}
 deform_pec_aim_base_l {parent=spine_04, loc=<7.8732 12.6388 -2.4123>, rot=<0.0102 -0.4602 0.3979 -0.7936>}
 deform_pec_aim_l {parent=deform_pec_aim_base_l, loc=<0 0 0>, rot=<0 0 0 1>}
 deform_pec_base_l {parent=deform_pec_aim_l, loc=<10.766 0 0>, rot=<0 0 0 1>}
 deform_pec_l {parent=deform_pec_base_l, loc=<0.055 0 0>, rot=<0 0 0 1>}
 dyn_l_pectoralDynamics_master {parent=deform_pec_l, loc=<0 0 0>, rot=<-0.2644 0.4911 0.803 0.2101>}
 dyn_l_pectoralDynamics {parent=dyn_l_pectoralDynamics_master, loc=<4.6227 0 0>, rot=<0 0 0 1>}
 deform_pec_aim_base_r {parent=spine_04, loc=<7.8727 12.6388 2.4123>, rot=<-0.4602 -0.0102 0.7936 0.3979>}
 deform_pec_aim_r {parent=deform_pec_aim_base_r, loc=<0 0 0>, rot=<0 0 0 1>}
 deform_pec_base_r {parent=deform_pec_aim_r, loc=<-10.7662 -0.0006 -0.0007>, rot=<0 0 0 1>}
 deform_pec_r {parent=deform_pec_base_r, loc=<0 0 0>, rot=<0 0 0 1>}
 dyn_r_pectoralDynamics_master {parent=deform_pec_r, loc=<0 0 0>, rot=<-0.135 0.3798 -0.0885 -0.9109>}
 dyn_r_pectoralDynamics {parent=dyn_r_pectoralDynamics_master, loc=<4.2261 0 0>, rot=<0 0 0 1>}
 Base_Bullet {parent=spine_04, loc=<2.4467 -1.7043 -0.0282>, rot=<-0.0113 0.2818 -0.0415 -0.9585>}
 Base_Upper_Bullet {parent=Base_Bullet, loc=<0 0 0>, rot=<0 0 0 1>}
 dyn_Upper_Clavical {parent=Base_Upper_Bullet, loc=<24.4938 1.0256 0>, rot=<0 0 -0.1299 -0.9915>}
 dyn_Front_Clavical {parent=Base_Upper_Bullet, loc=<13.2093 13.1985 0>, rot=<-0.2332 -0.0965 -0.3701 -0.894>}
 dyn_Back {parent=Base_Upper_Bullet, loc=<1.9991 -14.2493 -2.3938>, rot=<-0.0881 -0.0008 0.6978 -0.7109>}
 dyn_Front {parent=Base_Upper_Bullet, loc=<1.0723 18.6973 -0.1968>, rot=<-0.1883 -0.1847 -0.6612 -0.7023>}
 dyn_Back_Clavical {parent=Base_Upper_Bullet, loc=<14.544 -12.5486 -1.4391>, rot=<-0.0531 -0.0329 0.586 -0.8079>}
 deform_lat_l {parent=spine_04, loc=<0.5749 -9.0477 -12.5592>, rot=<0.7065 0.0294 0.7065 -0.0294>}
 deform_lat_r {parent=spine_04, loc=<0.5749 -9.0477 12.5592>, rot=<0.7065 0.0294 0.7065 -0.0294>}
 spine_05_null {parent=spine_04, type=Null, loc=<10.0146 0 0>, rot=<0 0 -0.0015 -1>}
 Base_Lower_Bullet {parent=spine_03, loc=<10.593 -1.6394 -0.0282>, rot=<-0.3503 -0.0981 0.9315 -0.0029>}
 dyn_Front_Lower {parent=Base_Lower_Bullet, loc=<12.8671 -13.1546 -3.5571>, rot=<-0.0099 -0.096 0.507 -0.8565>}
 dyn_Lower_Side {parent=Base_Lower_Bullet, loc=<20.142 0.341 -3.405>, rot=<0.0474 -0.2773 -0.003 -0.9596>}
 dyn_Back_Lower {parent=Base_Lower_Bullet, loc=<11.1665 12.5319 -1.7294>, rot=<0.0378 -0.0429 -0.547 -0.8352>}
 thigh_l {parent=pelvis, loc=<-2.5693 -0.0109 -7.7501>, rot=<0.0012 0.0572 0.0155 -0.9982>}
 calf_l {parent=thigh_l, loc=<-39.0484 0 0>, rot=<0.0006 -0.0161 -0.0391 -0.9991>}
 foot_l {parent=calf_l, loc=<-42.561 0.1077 0.4292>, rot=<0.0418 -0.0455 0.0343 -0.9975>}
 ball_l {parent=foot_l, loc=<-6.1599 13.8264 0.6806>, rot=<0.0001 -0.0001 -0.7063 -0.7079>}
 calf_twist_02_l {parent=calf_l, loc=<-18.8064 -0.0967 0.0397>, rot=<0.0237 0.0095 -0.0035 -0.9997>}
 calf_twist_01_l {parent=calf_l, loc=<-35.813 -0.0967 0.0397>, rot=<0.0237 0.0095 -0.0035 -0.9997>}
 dyn_left_a_shoe_laces_base {parent=calf_twist_01_l, loc=<7.2528 5.0447 -1.1102>, rot=<-0.0636 -0.8463 -0.5235 -0.0756>}
 dyn_left_a_shoe_laces_01 {parent=dyn_left_a_shoe_laces_base, loc=<0 0 0>, rot=<0 0 0 1>}
 dyn_left_a_shoe_laces_02 {parent=dyn_left_a_shoe_laces_01, loc=<3.5 0 0>, rot=<0 0 0 1>}
 dyn_left_b_shoe_laces_base {parent=calf_twist_01_l, loc=<7.2646 5.2971 -0.3032>, rot=<0.1594 0.6506 0.7221 -0.1731>}
 dyn_left_b_shoe_laces_01 {parent=dyn_left_b_shoe_laces_base, loc=<0 0 0>, rot=<0 0.0476 0 -0.9989>}
 dyn_left_b_shoe_laces_02 {parent=dyn_left_b_shoe_laces_01, loc=<3.5 0 0>, rot=<0 0 0 1>}
 deform_knee_in_base_l {parent=calf_l, loc=<1.298 -5.2957 -0.0496>, rot=<0 0 0 -1>}
 deform_knee_in_l {parent=deform_knee_in_base_l, loc=<0 0 0>, rot=<0 0 0 1>}
 deform_knee_out_base_l {parent=calf_l, loc=<0 4 0>, rot=<0 0 0 -1>}
 deform_knee_out_l {parent=deform_knee_out_base_l, loc=<0 0 0>, rot=<0 0 0 1>}
 thigh_twist_01_l {parent=thigh_l, loc=<-21.6303 0 0>, rot=<0.0013 0 0 -1>}
 deform_groin_base_l {parent=thigh_l, loc=<-2.1908 8.1142 -1.7442>, rot=<0 0 0 -1>}
 deform_groin_l {parent=deform_groin_base_l, loc=<0 0 0>, rot=<0 0 0 1>}
 thigh_r {parent=pelvis, loc=<-2.5693 -0.0109 7.7501>, rot=<0.0572 -0.0012 0.9982 0.0155>}
 calf_r {parent=thigh_r, loc=<39.0484 0 0>, rot=<0.0006 -0.0161 -0.0391 -0.9991>}
 foot_r {parent=calf_r, loc=<42.561 -0.1077 -0.4292>, rot=<0.0418 -0.0455 0.0343 -0.9975>}
 ball_r {parent=foot_r, loc=<6.1599 -13.8264 -0.6806>, rot=<0.0001 -0.0001 -0.7063 -0.7079>}
 calf_twist_02_r {parent=calf_r, loc=<18.8064 0.0967 -0.0397>, rot=<0.0237 0.0095 -0.0035 -0.9997>}
 calf_twist_01_r {parent=calf_r, loc=<35.813 0.0967 -0.0397>, rot=<0.0237 0.0095 -0.0035 -0.9997>}
 dyn_right_b_shoe_laces_base {parent=calf_twist_01_r, loc=<-7.1448 -5.2203 0.0896>, rot=<-0.7811 -0.1094 0.1968 -0.5824>}
 dyn_right_b_shoe_laces_01 {parent=dyn_right_b_shoe_laces_base, loc=<0 0 0>, rot=<0 0 0 1>}
 dyn_right_b_shoe_laces_02 {parent=dyn_right_b_shoe_laces_01, loc=<3.5 0 0>, rot=<0 0 0 1>}
 dyn_right_a_shoe_laces_base {parent=calf_twist_01_r, loc=<-7.2187 -5.082 1.0237>, rot=<-0.5318 0.0541 0.0504 -0.8437>}
 dyn_right_a_shoe_laces_01 {parent=dyn_right_a_shoe_laces_base, loc=<0 0 0>, rot=<0 0 0 1>}
 dyn_right_a_shoe_laces_02 {parent=dyn_right_a_shoe_laces_01, loc=<3.5 0 0>, rot=<0 0 0 1>}
 deform_knee_in_base_r {parent=calf_r, loc=<-1.298 5.2957 0.0496>, rot=<0 0 0 -1>}
 deform_knee_in_r {parent=deform_knee_in_base_r, loc=<0 0 0>, rot=<0 0 0 1>}
 deform_knee_out_base_r {parent=calf_r, loc=<0 -4 0>, rot=<0 0 0 -1>}
 deform_knee_out_r {parent=deform_knee_out_base_r, loc=<0 0 0>, rot=<0 0 0 1>}
 thigh_twist_01_r {parent=thigh_r, loc=<21.6303 0 0>, rot=<0.0013 0 0 -1>}
 deform_groin_base_r {parent=thigh_r, loc=<2.1908 -8.1142 1.7442>, rot=<0 0 0 -1>}
 deform_groin_r {parent=deform_groin_base_r, loc=<0 0 0>, rot=<0 0 0 1>}
 deform_glute_base_l {parent=pelvis, loc=<-8.8064 -8.8222 -8.4287>, rot=<0.0275 0.0501 -0.4472 -0.8926>}
 deform_glute_l {parent=deform_glute_base_l, loc=<0 0 0>, rot=<0 0 0 1>}
 deform_glute_base_r {parent=pelvis, loc=<-8.8062 -8.8222 8.4287>, rot=<0.0501 -0.0275 0.8926 -0.4472>}
 deform_glute_r {parent=deform_glute_base_r, loc=<0 0 0>, rot=<0 0 0 1>}
 attach {parent=root, loc=<0 0 0>, rot=<0 0 0 1>}
 ik_hand_root {parent=root, loc=<0 0 0>, rot=<0 0 0 1>}
 ik_hand_gun {parent=ik_hand_root, loc=<-51.9727 4.0883 102.6935>, rot=<0.6084 0.1793 0.4613 -0.6205>}
 ik_hand_r {parent=ik_hand_gun, loc=<0 0 0>, rot=<0 0 0 1>}
 ik_hand_l {parent=ik_hand_gun, loc=<53.0285 82.1781 35.2051>, rot=<-0.5102 -0.7906 -0.3387 0>}
 ik_foot_root {parent=root, loc=<0 0 0>, rot=<0 0 0 1>}
 ik_foot_l {parent=ik_foot_root, loc=<15.2608 -4.4176 9.4432>, rot=<0.0357 0.7045 0.0284 -0.7082>}
 ik_foot_r {parent=ik_foot_root, loc=<-15.2608 -4.4176 9.4432>, rot=<0.7082 0.0284 -0.7045 0.0357>}
 root_fp {parent=root, loc=<0 0 0>, rot=<0 0 0 1>}
 attach_fp {parent=root_fp, loc=<0 0 0>, rot=<0 0 0 1>}
 camera_root {parent=root_fp, loc=<0 0 0>, rot=<0 0 0 1>}
 camera {parent=camera_root, loc=<0 0 0>, rot=<0 0 0 1>}
 attach_emote_01 {parent=root, loc=<0 0 0>, rot=<0 0 0 1>}
 attach_emote_02 {parent=root, loc=<0 0 0>, rot=<0 0 0 1>}
 attach_emote_03 {parent=root, loc=<0 0 0>, rot=<0 0 0 1>}
 warp_point_01 {parent=root, loc=<0 0 0>, rot=<0 0 0 1>}
 warp_point_02 {parent=root, loc=<0 0 0>, rot=<0 0 0 1>}
 warp_point_03 {parent=root, loc=<0 0 0>, rot=<0 0 0 1>}
]
variables [
 BackPackBelt {type=bool, default=False}
 Crouch {type=bool, default=False}
 FrontEnd {type=bool, default=False}
 GroundMovement {type=double, default=0}
 IsJumping {type=double, default=0}
 ParachuteOpened {type=bool, default=False}
 PlayingEmote {type=bool, default=False}
 Skydiving {type=bool, default=False}
 TankTopOff {type=bool, default=False}
 Targeting AIm Up {type=bool, default=False}
]
