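# Physics asset for the RaisinHunt Rag medium male face accessory, providing capsule bodies for dynamic hair over a head, neck, upper torso and upper arm rig.
# Body capsules and five hair-chain bases are angularly locked along the hierarchy.
# Each chain's swing limits widen from root to tip with tight twist, and right_a adds a branch.
# Angle-free cross-constraints pair neighboring strands: left_a to left, right_a to middle and right.
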PhysicsAsset /BRCosmetics/Characters/Player/Male/Medium/Bodies/M_MED_RaisinHunt_Rag/Meshes/Parts/M_MED_RaisinHunt_Rag_FaceAcc_Physics
bodies {
 clavicle_l [{shapeType=Capsule, length=5.7508, radius=9.1209, center=<9.4046 -1.5262 0.0143>, rotation=<rot -70.1162 -167.158 -6.4361>}]
 clavicle_r [{shapeType=Capsule, length=5.7508, radius=9.1209, center=<-9.4046 1.5262 -0.0143>, rotation=<rot -70.1162 -167.158 173.5639>}]
 dyn_left_hair_01 [{shapeType=Capsule, length=3, radius=2.2, shapeName=dyn_left_hair_01_capsule, center=<1.5 0 0>, rotation=<rot -90 0 0>}]
 dyn_left_hair_02 [{shapeType=Capsule, length=3, radius=2.2, shapeName=dyn_left_hair_02_capsule, center=<2 0 0>, rotation=<rot -90 0 0>}]
 dyn_left_hair_03 [{shapeType=Capsule, length=3, radius=2.2, shapeName=dyn_left_hair_03_capsule, center=<3 0 0>, rotation=<rot -90 0 0>}]
 dyn_left_hair_a_01 [{shapeType=Capsule, length=3, radius=2.2, shapeName=dyn_left_hair_a_01_capsule, center=<1.5 0 0>, rotation=<rot -90 0 0>}]
 dyn_left_hair_a_02 [{shapeType=Capsule, length=3, radius=2.2, shapeName=dyn_left_hair_a_02_capsule, center=<2 0 0>, rotation=<rot -90 0 0>}]
 dyn_left_hair_a_03 [{shapeType=Capsule, length=3, radius=2.2, shapeName=dyn_left_hair_a_03_capsule, center=<3 0 2.515>, rotation=<rot -90 9.1953 0>}]
 dyn_left_hair_a_base [{shapeType=Capsule, length=0.505, radius=0.505, shapeName=dyn_left_hair_a_base_capsule, rotation=<rot 0 -90 0>}]
 dyn_left_hair_base [{shapeType=Capsule, length=0.505, radius=0.505, shapeName=dyn_left_hair_base_capsule, rotation=<rot 0 -90 0>}]
 dyn_middle_hair_01 [{shapeType=Capsule, length=3, radius=2.2, shapeName=dyn_middle_hair_01_capsule, center=<1.5 0 0>, rotation=<rot -90 0 0>}]
 dyn_middle_hair_02 [{shapeType=Capsule, length=3, radius=2.2, shapeName=dyn_middle_hair_02_capsule, center=<2 0 0>, rotation=<rot -90 0 0>}]
 dyn_middle_hair_03 [{shapeType=Capsule, length=3, radius=2.2, shapeName=dyn_middle_hair_03_capsule, center=<3 0 0>, rotation=<rot -90 0 0>}]
 dyn_middle_hair_base [{shapeType=Capsule, length=0.505, radius=0.505, shapeName=dyn_middle_hair_base_capsule, rotation=<rot 0 -90 0>}]
 dyn_right_hair_01 [{shapeType=Capsule, length=3, radius=2.2, shapeName=dyn_right_hair_01_capsule, center=<1.5 0 0>, rotation=<rot -90 0 0>}]
 dyn_right_hair_02 [{shapeType=Capsule, length=3, radius=2.2, shapeName=dyn_right_hair_02_capsule, center=<2 0 0>, rotation=<rot -90 0 0>}]
 dyn_right_hair_03 [{shapeType=Capsule, length=3, radius=2.2, shapeName=dyn_right_hair_03_capsule, center=<3 0 0>, rotation=<rot -90 0 0>}]
 dyn_right_hair_a_01 [{shapeType=Capsule, length=3, radius=2.2, shapeName=dyn_right_hair_a_01_capsule, center=<1.5 0 0>, rotation=<rot -90 0 0>}]
 dyn_right_hair_a_02 [
  {shapeType=Capsule, length=3.5869, radius=2.2652, shapeName=dyn_right_hair_a_02_capsule, center=<2 0 0>, rotation=<rot -90 0 0>}
  {
   shapeType=Capsule
   length=5.6426
   radius=2.3574
   shapeName=dyn_right_hair_a_02_capsule
   center=<0.9353 -1.2735 -1.1169>
   rotation=<rot 2.5461 50.4854 -111.2484>
  }
 ]
 dyn_right_hair_a_03 [{shapeType=Capsule, length=2.0719, radius=1.8429, shapeName=dyn_right_hair_a_03_capsule, center=<3 0 0>, rotation=<rot -90 0 0>}]
 dyn_right_hair_a_05 [{shapeType=Capsule, length=5.3201, radius=1.6333, shapeName=dyn_right_hair_a_05_capsule, center=<6.5 0 0>, rotation=<rot -90 0 0>}]
 dyn_right_hair_a_base [{shapeType=Capsule, length=0.505, radius=0.505, shapeName=dyn_right_hair_a_base_capsule, rotation=<rot 0 -90 0>}]
 dyn_right_hair_base [{shapeType=Capsule, length=0.505, radius=0.505, shapeName=dyn_right_hair_base_capsule, rotation=<rot 0 -90 0>}]
 head [{shapeType=Capsule, length=7.1864, radius=4.6625, center=<4.0708 0 0>, rotation=<rot -90 0 0>}]
 neck_01 [{shapeType=Capsule, length=2.4813, radius=5.7334, shapeName=neck_01_capsule, rotation=<rot 0 -90 90>}]
 neck_02 [{shapeType=Capsule, length=3.8915, radius=5.1764, shapeName=neck_02_capsule, rotation=<rot 0 -90 90>}]
 spine_04 [{shapeType=Capsule, length=9.6612, radius=13.7526, center=<5.6117 -0.6531 0>}]
 spine_05 [{shapeType=Capsule, length=12.6473, radius=7.3804, center=<5.4473 -1.24 0>}]
 upperarm_twist_01_l [{shapeType=Capsule, length=26.1978, radius=5.879, center=<12 0 0>, rotation=<rot -90 0 0>}]
 upperarm_twist_01_r [{shapeType=Capsule, length=22, radius=5.5, center=<-12 0 0>, rotation=<rot -90 0 0>}]
}
constraints [
 {
  bone1Name=spine_05
  bone2Name=spine_04
  swing1LimitDegrees=45
  swing1Motion=Locked
  swing2LimitDegrees=45
  swing2Motion=Locked
  twistLimitDegrees=45
  twistMotion=Locked
 }
 {
  bone1Name=clavicle_l
  bone2Name=spine_05
  swing1LimitDegrees=45
  swing1Motion=Locked
  swing2LimitDegrees=45
  swing2Motion=Locked
  twistLimitDegrees=45
  twistMotion=Locked
 }
 {
  bone1Name=clavicle_r
  bone2Name=spine_05
  swing1LimitDegrees=45
  swing1Motion=Locked
  swing2LimitDegrees=45
  swing2Motion=Locked
  twistLimitDegrees=45
  twistMotion=Locked
 }
 {
  bone1Name=neck_01
  bone2Name=spine_05
  swing1LimitDegrees=45
  swing1Motion=Locked
  swing2LimitDegrees=45
  swing2Motion=Locked
  twistLimitDegrees=45
  twistMotion=Locked
 }
 {
  bone1Name=upperarm_twist_01_l
  bone2Name=clavicle_l
  swing1LimitDegrees=45
  swing1Motion=Locked
  swing2LimitDegrees=45
  swing2Motion=Locked
  twistLimitDegrees=45
  twistMotion=Locked
 }
 {
  bone1Name=upperarm_twist_01_r
  bone2Name=clavicle_r
  swing1LimitDegrees=45
  swing1Motion=Locked
  swing2LimitDegrees=45
  swing2Motion=Locked
  twistLimitDegrees=45
  twistMotion=Locked
 }
 {
  bone1Name=neck_02
  bone2Name=neck_01
  swing1LimitDegrees=45
  swing1Motion=Locked
  swing2LimitDegrees=45
  swing2Motion=Locked
  twistLimitDegrees=45
  twistMotion=Locked
 }
 {
  bone1Name=head
  bone2Name=neck_02
  swing1LimitDegrees=45
  swing1Motion=Locked
  swing2LimitDegrees=45
  swing2Motion=Locked
  twistLimitDegrees=45
  twistMotion=Locked
 }
 {
  bone1Name=dyn_left_hair_a_base
  bone2Name=head
  swing1LimitDegrees=45
  swing1Motion=Locked
  swing2LimitDegrees=45
  swing2Motion=Locked
  twistLimitDegrees=45
  twistMotion=Locked
 }
 {
  bone1Name=dyn_left_hair_base
  bone2Name=head
  swing1LimitDegrees=45
  swing1Motion=Locked
  swing2LimitDegrees=45
  swing2Motion=Locked
  twistLimitDegrees=45
  twistMotion=Locked
 }
 {
  bone1Name=dyn_middle_hair_base
  bone2Name=head
  swing1LimitDegrees=45
  swing1Motion=Locked
  swing2LimitDegrees=45
  swing2Motion=Locked
  twistLimitDegrees=45
  twistMotion=Locked
 }
 {
  bone1Name=dyn_right_hair_a_base
  bone2Name=head
  swing1LimitDegrees=45
  swing1Motion=Locked
  swing2LimitDegrees=45
  swing2Motion=Locked
  twistLimitDegrees=45
  twistMotion=Locked
 }
 {
  bone1Name=dyn_right_hair_base
  bone2Name=head
  swing1LimitDegrees=45
  swing1Motion=Locked
  swing2LimitDegrees=45
  swing2Motion=Locked
  twistLimitDegrees=45
  twistMotion=Locked
 }
 {bone1Name=dyn_left_hair_a_01, bone2Name=dyn_left_hair_a_base, swing1LimitDegrees=15, swing2LimitDegrees=15, twistLimitDegrees=2}
 {bone1Name=dyn_left_hair_a_02, bone2Name=dyn_left_hair_a_01, swing1LimitDegrees=30, swing2LimitDegrees=30, twistLimitDegrees=2}
 {
  bone1Name=dyn_left_hair_a_01
  bone2Name=dyn_left_hair_01
  swing1LimitDegrees=45
  swing1Motion=Free
  swing2LimitDegrees=45
  swing2Motion=Free
  twistLimitDegrees=45
  twistMotion=Free
 }
 {bone1Name=dyn_left_hair_a_03, bone2Name=dyn_left_hair_a_02, swing1LimitDegrees=40, swing2LimitDegrees=40, twistLimitDegrees=2}
 {
  bone1Name=dyn_left_hair_a_02
  bone2Name=dyn_left_hair_02
  swing1LimitDegrees=45
  swing1Motion=Free
  swing2LimitDegrees=45
  swing2Motion=Free
  twistLimitDegrees=45
  twistMotion=Free
 }
 {
  bone1Name=dyn_left_hair_a_03
  bone2Name=dyn_left_hair_03
  swing1LimitDegrees=45
  swing1Motion=Free
  swing2LimitDegrees=45
  swing2Motion=Free
  twistLimitDegrees=45
  twistMotion=Free
 }
 {bone1Name=dyn_left_hair_01, bone2Name=dyn_left_hair_base, swing1LimitDegrees=15, swing2LimitDegrees=15, twistLimitDegrees=2}
 {bone1Name=dyn_left_hair_02, bone2Name=dyn_left_hair_01, swing1LimitDegrees=30, swing2LimitDegrees=30, twistLimitDegrees=2}
 {bone1Name=dyn_left_hair_03, bone2Name=dyn_left_hair_02, swing1LimitDegrees=40, swing2LimitDegrees=40, twistLimitDegrees=2}
 {bone1Name=dyn_middle_hair_01, bone2Name=dyn_middle_hair_base, swing1LimitDegrees=15, swing2LimitDegrees=15, twistLimitDegrees=2}
 {bone1Name=dyn_middle_hair_02, bone2Name=dyn_middle_hair_01, swing1LimitDegrees=30, swing2LimitDegrees=30, twistLimitDegrees=2}
 {
  bone1Name=dyn_right_hair_a_01
  bone2Name=dyn_middle_hair_01
  swing1LimitDegrees=45
  swing1Motion=Free
  swing2LimitDegrees=45
  swing2Motion=Free
  twistLimitDegrees=45
  twistMotion=Free
 }
 {bone1Name=dyn_middle_hair_03, bone2Name=dyn_middle_hair_02, swing1LimitDegrees=40, swing2LimitDegrees=40, twistLimitDegrees=2}
 {
  bone1Name=dyn_right_hair_a_02
  bone2Name=dyn_middle_hair_02
  swing1LimitDegrees=45
  swing1Motion=Free
  swing2LimitDegrees=45
  swing2Motion=Free
  twistLimitDegrees=45
  twistMotion=Free
 }
 {
  bone1Name=dyn_right_hair_a_03
  bone2Name=dyn_middle_hair_03
  swing1LimitDegrees=45
  swing1Motion=Free
  swing2LimitDegrees=45
  swing2Motion=Free
  twistLimitDegrees=45
  twistMotion=Free
 }
 {bone1Name=dyn_right_hair_a_01, bone2Name=dyn_right_hair_a_base, swing1LimitDegrees=15, swing2LimitDegrees=15, twistLimitDegrees=2}
 {bone1Name=dyn_right_hair_a_02, bone2Name=dyn_right_hair_a_01, swing1LimitDegrees=30, swing2LimitDegrees=30, twistLimitDegrees=2}
 {
  bone1Name=dyn_right_hair_a_01
  bone2Name=dyn_right_hair_01
  swing1LimitDegrees=45
  swing1Motion=Free
  swing2LimitDegrees=45
  swing2Motion=Free
  twistLimitDegrees=45
  twistMotion=Free
 }
 {bone1Name=dyn_right_hair_a_03, bone2Name=dyn_right_hair_a_02, swing1LimitDegrees=40, swing2LimitDegrees=40, twistLimitDegrees=2}
 {bone1Name=dyn_right_hair_a_05, bone2Name=dyn_right_hair_a_02, swing1LimitDegrees=20, swing2LimitDegrees=20, twistLimitDegrees=2}
 {
  bone1Name=dyn_right_hair_a_02
  bone2Name=dyn_right_hair_02
  swing1LimitDegrees=45
  swing1Motion=Free
  swing2LimitDegrees=45
  swing2Motion=Free
  twistLimitDegrees=45
  twistMotion=Free
 }
 {
  bone1Name=dyn_right_hair_a_03
  bone2Name=dyn_right_hair_03
  swing1LimitDegrees=45
  swing1Motion=Free
  swing2LimitDegrees=45
  swing2Motion=Free
  twistLimitDegrees=45
  twistMotion=Free
 }
 {bone1Name=dyn_right_hair_01, bone2Name=dyn_right_hair_base, swing1LimitDegrees=15, swing2LimitDegrees=15, twistLimitDegrees=2}
 {bone1Name=dyn_right_hair_02, bone2Name=dyn_right_hair_01, swing1LimitDegrees=30, swing2LimitDegrees=30, twistLimitDegrees=2}
 {bone1Name=dyn_right_hair_03, bone2Name=dyn_right_hair_02, swing1LimitDegrees=40, swing2LimitDegrees=40, twistLimitDegrees=2}
]
