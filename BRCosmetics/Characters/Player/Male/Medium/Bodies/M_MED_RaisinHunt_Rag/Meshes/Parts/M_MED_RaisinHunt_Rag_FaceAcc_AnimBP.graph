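# Post-processes the RaisinHunt Rag character's face accessory, simulating hair and headband straps.
# Copies the face mesh pose through a control rig, rigid-body physics and hair constraints into a cached pose.
# That cache feeds band simulations switched by backpack-belt and front-end flags, then a state-driven control rig and per-bone AnimDynamics springs.
# The event graph updates gravity parameters, cosmetic variant flags, jump state and physics alphas; both Trail controllers have zero alpha.
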AnimBlueprint /BRCosmetics/Characters/Player/Male/Medium/Bodies/M_MED_RaisinHunt_Rag/Meshes/Parts/M_MED_RaisinHunt_Rag_FaceAcc_AnimBP
target_skeleton=&/M_MED_RaisinHunt_Rag_FaceAcc_Skeleton
graphs [
 AnimGraph {nodes [
   n1 = RigidBody!(
    ComponentPose=Local To Component(LocalPose=Use cached pose 'BAse_01'())
    Alpha
    OverrideWorldGravity:Vector="0,0,0"
    ExternalForce
    SimSpaceSettings:Sim Space Settings Structure=(WorldAlpha=1,VelocityScaleZ=1,DampingAlpha=1,MaxLinearVelocity=400,MaxAngularVelocity=40,MaxLinearAcceleration=2000,MaxAngularAcceleration=200,ExternalLinearDragV=(X=5,Y=5,Z=0),ExternalLinearVelocity=(X=0,Y=0,Z=0),ExternalAngularVelocity=(X=0,Y=0,Z=0))
   ) -> (Pose) {
    bOverrideWorldGravity=true
    overridePhysicsAsset=&/M_MED_RaisinHunt_Rag_FaceAcc_Band_Physics
    simulationSpace=BaseBoneSpace
    baseBoneRef {boneName=spine_05}
    simSpaceSettings {
     maxAngularAcceleration=200
     maxAngularVelocity=40
     maxLinearAcceleration=2000
     maxLinearVelocity=400
     worldAlpha=1
     externalLinearDragV=<5 5>
    }
   }
   n2 = RigidBody(
    ComponentPose=Local To Component(LocalPose=Use cached pose 'BAse_01'())
    OverrideWorldGravity="0,0,0"
    SimSpaceSettings=(WorldAlpha=1,VelocityScaleZ=1,DampingAlpha=1,MaxLinearVelocity=400,MaxAngularVelocity=40,MaxLinearAcceleration=2000,MaxAngularAcceleration=200,ExternalLinearDragV=(X=5,Y=5,Z=0),ExternalLinearVelocity=(X=0,Y=0,Z=0),ExternalAngularVelocity=(X=0,Y=0,Z=0))
   ) {
    bOverrideWorldGravity=true
    overridePhysicsAsset=&/M_MED_RaisinHunt_Rag_FaceAcc_Band_Physics_02
    simulationSpace=BaseBoneSpace
    baseBoneRef {boneName=spine_05}
    simSpaceSettings {
     maxAngularAcceleration=200
     maxAngularVelocity=40
     maxLinearAcceleration=2000
     maxLinearVelocity=400
     worldAlpha=1
     externalLinearDragV=<5 5>
    }
   }
   n3 = Component To Local(ComponentPose=@n2)
   n4 = RigidBody(
    ComponentPose=Local To Component(LocalPose=Use cached pose 'BAse_01'())
    OverrideWorldGravity="0,0,0"
    SimSpaceSettings=(WorldAlpha=1,VelocityScaleZ=1,DampingAlpha=1,MaxLinearVelocity=400,MaxAngularVelocity=40,MaxLinearAcceleration=2000,MaxAngularAcceleration=200,ExternalLinearDragV=(X=5,Y=5,Z=0),ExternalLinearVelocity=(X=0,Y=0,Z=0),ExternalAngularVelocity=(X=0,Y=0,Z=0))
   ) {
    bOverrideWorldGravity=true
    overridePhysicsAsset=&/M_MED_RaisinHunt_Rag_FaceAcc_Band_Physics_FE
    simulationSpace=BaseBoneSpace
    baseBoneRef {boneName=spine_05}
    simSpaceSettings {
     maxAngularAcceleration=200
     maxAngularVelocity=40
     maxLinearAcceleration=2000
     maxLinearVelocity=400
     worldAlpha=1
     externalLinearDragV=<5 5>
    }
   }
   n5 = Component To Local(ComponentPose=@n4)
   n6 = Get bIsInFrontEnd()
   n7 = Trail controller - Bone: dyn_headband_straps_b_04(ComponentPose=Local To Component(LocalPose=Blend Poses by bool(BlendPose_0=@n5, BlendPose_1=@n3, bActiveValue=@n6)), Alpha=0) {
    chainLength=3
    relaxationSpeedScale=2
    rotationLimits [{limitMax=<180 180 180>}, {limitMax=<180 180 180>}, {limitMax=<180 180 180>}]
    rotationOffsets [<0 0 0>, <0 0 0>, <0 0 0>]
    trailBone {boneName=dyn_headband_straps_b_04}
    alphaBoolBlend=~
   }
   n8 = Control Rig!(
    Source=Blend Poses by bool(
     BlendPose_0=Component To Local(ComponentPose=@n7)
     BlendPose_1=Component To Local(ComponentPose=@n1)
     bActiveValue=Get BeltForBackpack() -> (BeltForBackpack:bool)
    )
    Alpha
    Targeting AIm Up:bool=Get bIsTargeting()
    Crouch:bool=Get bIsCrouching()
    Skydiving:bool=Get bIsSkyDiving()
    IsJumping:double=Get IsJumping() -> (IsJumping:double)
    FrontEnd:bool=Get bIsInFrontEnd()
    ParachuteOpened:bool=Get bIsParachuteOpened()
    BackPackBelt:bool=Get BeltForBackpack()
    PlayingEmote:bool=Get bIsPlayingEmote()
   ) -> (Pose) {controlRigAssetReference {blueprintRigClass=&/M_MED_RaisinHunt_Rag_FaceAcc_CtrlRig_Post.M_MED_RaisinHunt_Rag_FaceAcc_CtrlRig_Post_C}}
   n9 = Anim Dynamics - Bone: dyn_front_l_hair_01!(ComponentPose=Local To Component(LocalPose=@n8), Alpha=1, GravityOverride:Vector="0,0,-980", ExternalForce="0,0,0") -> (Pose) {
    angularDampingOverride=1
    angularSpringConstant=18000
    bAngularSpring=true
    bOverrideAngularDamping=true
    bUseGravityOverride=true
    gravityScale=0.35
    simulationSpace=RootRelative
    boundBone {boneName=dyn_front_l_hair_01}
    componentAppliedLinearAccClamp=<2000 2000 2000>
    componentLinearAccScale=<0.2 0.2 0.2>
    componentLinearVelScale=<0.2 0.2 0.2>
    gravityOverride {z=-980}
    physicsBodyDefinitions [{
      boundBone {boneName=dyn_front_l_hair_01}
      boxExtents=<20 30 30>
      constraintSetup {angularLimitsMax=<0 0 15>, angularLimitsMin=<0 -15 -15>, angularTarget=<1 0 0>}
      localJointOffset=<5 0 0>
     }]
    simSpaceSettings {maxAngularAcceleration=40, maxAngularVelocity=200, simSpaceAngularAlpha=0.6}
   }
   n10 = Anim Dynamics - Bone: dyn_front_r_hair_01!(ComponentPose=@n9, Alpha=1, GravityOverride:Vector="0,0,-980", ExternalForce="0,0,0") -> (Pose) {
    angularDampingOverride=1
    angularSpringConstant=18000
    bAngularSpring=true
    bOverrideAngularDamping=true
    bUseGravityOverride=true
    gravityScale=0.35
    simulationSpace=RootRelative
    boundBone {boneName=dyn_front_r_hair_01}
    componentAppliedLinearAccClamp=<2000 2000 2000>
    componentLinearAccScale=<0.2 0.2 0.2>
    componentLinearVelScale=<0.2 0.2 0.2>
    gravityOverride {z=-980}
    physicsBodyDefinitions [{
      boundBone {boneName=dyn_front_r_hair_01}
      boxExtents=<20 30 30>
      constraintSetup {angularLimitsMax=<0 0 15>, angularLimitsMin=<0 -15 -15>, angularTarget=<1 0 0>}
      localJointOffset=<5 0 0>
     }]
    simSpaceSettings {maxAngularAcceleration=40, maxAngularVelocity=200, simSpaceAngularAlpha=0.6}
   }
   n11 = Anim Dynamics - Bone: dyn_right_hair_c_01!(ComponentPose=@n10, Alpha=1, GravityOverride:Vector="0,0,-980", ExternalForce="0,0,0") -> (Pose) {
    angularDampingOverride=1
    angularSpringConstant=18000
    bAngularSpring=true
    bOverrideAngularDamping=true
    bUseGravityOverride=true
    gravityScale=0.35
    simulationSpace=RootRelative
    boundBone {boneName=dyn_right_hair_c_01}
    componentAppliedLinearAccClamp=<2000 2000 2000>
    componentLinearAccScale=<0.6 0.6 0.6>
    componentLinearVelScale=<0.6 0.6 0.6>
    gravityOverride {z=-980}
    physicsBodyDefinitions [{
      boundBone {boneName=dyn_right_hair_c_01}
      boxExtents=<20 30 30>
      constraintSetup {angularLimitsMax=<0 0 2>, angularLimitsMin=<0 -15 -2>, angularTarget=<1 0 0>}
      localJointOffset=<5 0 0>
     }]
    simSpaceSettings {maxAngularAcceleration=40, maxAngularVelocity=200, simSpaceAngularAlpha=0.6}
   }
   n12 = Anim Dynamics - Bone: dyn_left_hair_c_01!(ComponentPose=@n11, Alpha=1, GravityOverride:Vector="0,0,-980", ExternalForce="0,0,0") -> (Pose) {
    angularDampingOverride=1
    angularSpringConstant=18000
    bAngularSpring=true
    bOverrideAngularDamping=true
    bUseGravityOverride=true
    gravityScale=0.35
    simulationSpace=RootRelative
    boundBone {boneName=dyn_left_hair_c_01}
    componentAppliedLinearAccClamp=<2000 2000 2000>
    componentLinearAccScale=<0.2 0.2 0.2>
    componentLinearVelScale=<0.2 0.2 0.2>
    gravityOverride {z=-980}
    physicsBodyDefinitions [{
      boundBone {boneName=dyn_left_hair_c_01}
      boxExtents=<20 30 30>
      constraintSetup {angularLimitsMax=<0 0 15>, angularLimitsMin=<0 -15 -15>, angularTarget=<1 0 0>}
      localJointOffset=<5 0 0>
     }]
    simSpaceSettings {maxAngularAcceleration=40, maxAngularVelocity=200, simSpaceAngularAlpha=0.6}
   }
   n13 = Anim Dynamics - Bone: dyn_back_hair_a_01!(ComponentPose=@n12, Alpha=1, GravityOverride:Vector="0,0,-980", ExternalForce="0,0,0") -> (Pose) {
    angularDampingOverride=1
    angularSpringConstant=18000
    bAngularSpring=true
    bOverrideAngularDamping=true
    bUseGravityOverride=true
    gravityScale=0.35
    simulationSpace=RootRelative
    boundBone {boneName=dyn_back_hair_a_01}
    componentAppliedLinearAccClamp=<2000 2000 2000>
    componentLinearAccScale=<0.2 0.2 0.2>
    componentLinearVelScale=<0.2 0.2 0.2>
    gravityOverride {z=-980}
    physicsBodyDefinitions [{
      boundBone {boneName=dyn_back_hair_a_01}
      boxExtents=<20 30 30>
      constraintSetup {angularLimitsMax=<0 0 15>, angularLimitsMin=<0 -15 -15>, angularTarget=<1 0 0>}
      localJointOffset=<5 0 0>
     }]
    simSpaceSettings {maxAngularAcceleration=40, maxAngularVelocity=200, simSpaceAngularAlpha=0.6}
   }
   n14 = Anim Dynamics - Bone: dyn_back_hair_b_01!(ComponentPose=@n13, Alpha=1, GravityOverride:Vector="0,0,-980", ExternalForce="0,0,0") -> (Pose) {
    angularDampingOverride=1
    angularSpringConstant=18000
    bAngularSpring=true
    bOverrideAngularDamping=true
    bUseGravityOverride=true
    gravityScale=0.35
    simulationSpace=RootRelative
    boundBone {boneName=dyn_back_hair_b_01}
    componentAppliedLinearAccClamp=<2000 2000 2000>
    componentLinearAccScale=<0.2 0.2 0.2>
    componentLinearVelScale=<0.2 0.2 0.2>
    gravityOverride {z=-980}
    physicsBodyDefinitions [{
      boundBone {boneName=dyn_back_hair_b_01}
      boxExtents=<20 30 30>
      constraintSetup {angularLimitsMax=<0 0 15>, angularLimitsMin=<0 -15 -15>, angularTarget=<1 0 0>}
      localJointOffset=<5 0 0>
     }]
    simSpaceSettings {maxAngularAcceleration=40, maxAngularVelocity=200, simSpaceAngularAlpha=0.6}
   }
   n15 = Anim Dynamics - Bone: dyn_left_hair_b_01!(ComponentPose=@n14, Alpha=1, GravityOverride:Vector="0,0,-980", ExternalForce="0,0,0") -> (Pose) {
    angularDampingOverride=1
    angularSpringConstant=18000
    bAngularSpring=true
    bOverrideAngularDamping=true
    bUseGravityOverride=true
    gravityScale=0.35
    simulationSpace=RootRelative
    boundBone {boneName=dyn_left_hair_b_01}
    componentAppliedLinearAccClamp=<2000 2000 2000>
    componentLinearAccScale=<0.2 0.2 0.2>
    componentLinearVelScale=<0.2 0.2 0.2>
    gravityOverride {z=-980}
    physicsBodyDefinitions [{
      boundBone {boneName=dyn_left_hair_b_01}
      boxExtents=<20 30 30>
      constraintSetup {angularLimitsMax=<0 0 15>, angularLimitsMin=<0 -15 -15>, angularTarget=<1 0 0>}
      localJointOffset=<5 0 0>
     }]
    simSpaceSettings {maxAngularAcceleration=40, maxAngularVelocity=200, simSpaceAngularAlpha=0.6}
   }
   n16 = Anim Dynamics - Bone: dyn_right_hair_b_01!(ComponentPose=@n15, Alpha=1, GravityOverride:Vector="0,0,-980", ExternalForce="0,0,0") -> (Pose) {
    angularDampingOverride=1
    angularSpringConstant=18000
    bAngularSpring=true
    bOverrideAngularDamping=true
    bUseGravityOverride=true
    gravityScale=0.35
    simulationSpace=RootRelative
    boundBone {boneName=dyn_right_hair_b_01}
    componentAppliedLinearAccClamp=<2000 2000 2000>
    componentLinearAccScale=<0.2 0.2 0.2>
    componentLinearVelScale=<0.2 0.2 0.2>
    gravityOverride {z=-980}
    physicsBodyDefinitions [{
      boundBone {boneName=dyn_right_hair_b_01}
      boxExtents=<20 30 30>
      constraintSetup {angularLimitsMax=<0 0 15>, angularLimitsMin=<0 -15 -15>, angularTarget=<1 0 0>}
      localJointOffset=<5 0 0>
     }]
    simSpaceSettings {maxAngularAcceleration=40, maxAngularVelocity=200, simSpaceAngularAlpha=0.6}
   }
   n17 = Anim Dynamics - Bone: dyn_right_hair_a_05!(ComponentPose=@n16, Alpha=1, GravityOverride:Vector="0,0,-980", ExternalForce="0,0,0") -> (Pose) {
    angularDampingOverride=1
    angularSpringConstant=10000
    bAngularSpring=true
    bOverrideAngularDamping=true
    bUseGravityOverride=true
    gravityScale=0.35
    simulationSpace=RootRelative
    boundBone {boneName=dyn_right_hair_a_05}
    componentAppliedLinearAccClamp=<2000 2000 2000>
    componentLinearAccScale=<1 1 1>
    componentLinearVelScale=<1 1 1>
    gravityOverride {z=-980}
    physicsBodyDefinitions [{
      boundBone {boneName=dyn_right_hair_a_05}
      boxExtents=<40 30 30>
      constraintSetup {angularLimitsMax=<0 15 15>, angularLimitsMin=<0 -15 -15>, angularTarget=<1 0 0>}
      localJointOffset=<10 0 0>
     }]
   }
   n18 = Get FEBackpackANim() -> (FEBackpackANim:double)
   n19 = Anim Dynamics - Bone: dyn_headband_straps_a_02!(ComponentPose=@n17, Alpha=@n18, GravityOverride:Vector="0,0,-980", ExternalForce="0,0,0") -> (Pose) {
    angularDampingOverride=1
    angularSpringConstant=8000
    bAngularSpring=true
    bOverrideAngularDamping=true
    bUseGravityOverride=true
    gravityScale=0.35
    simulationSpace=RootRelative
    boundBone {boneName=dyn_headband_straps_a_02}
    componentAppliedLinearAccClamp=<2000 2000 2000>
    componentLinearAccScale=<0.5 0.5 0.5>
    componentLinearVelScale=<0.5 0.5 0.5>
    gravityOverride {z=-980}
    physicsBodyDefinitions [{
      boundBone {boneName=dyn_headband_straps_a_02}
      boxExtents=<20 30 30>
      constraintSetup {angularLimitsMax=<0 5 15>, angularLimitsMin=<0 -15 0>, angularTarget=<1 0 0>}
      localJointOffset=<5 0 0>
     }]
    simSpaceSettings {maxAngularAcceleration=40, maxAngularVelocity=200, simSpaceAngularAlpha=0.6}
   }
   n20 = Anim Dynamics - Bone: dyn_headband_straps_a_03!(ComponentPose=@n19, Alpha=@n18, GravityOverride:Vector="0,0,-980", ExternalForce="0,0,0") -> (Pose) {
    angularDampingOverride=1
    angularSpringConstant=8000
    bAngularSpring=true
    bOverrideAngularDamping=true
    bUseGravityOverride=true
    gravityScale=0.35
    simulationSpace=RootRelative
    boundBone {boneName=dyn_headband_straps_a_03}
    componentAppliedLinearAccClamp=<2000 2000 2000>
    componentLinearAccScale=<0.5 0.5 0.5>
    componentLinearVelScale=<0.5 0.5 0.5>
    gravityOverride {z=-980}
    physicsBodyDefinitions [{
      boundBone {boneName=dyn_headband_straps_a_03}
      boxExtents=<20 30 30>
      constraintSetup {angularLimitsMax=<0 5 15>, angularLimitsMin=<0 -15 0>, angularTarget=<1 0 0>}
      localJointOffset=<5 0 0>
     }]
    simSpaceSettings {maxAngularAcceleration=40, maxAngularVelocity=200, simSpaceAngularAlpha=0.6}
   }
   n21 = Anim Dynamics - Bone: dyn_headband_straps_a_04!(ComponentPose=@n20, Alpha=@n18, GravityOverride:Vector="0,0,-980", ExternalForce="0,0,0") -> (Pose) {
    angularDampingOverride=1
    angularSpringConstant=8000
    bAngularSpring=true
    bOverrideAngularDamping=true
    bUseGravityOverride=true
    gravityScale=0.35
    simulationSpace=RootRelative
    boundBone {boneName=dyn_headband_straps_a_04}
    componentAppliedLinearAccClamp=<2000 2000 2000>
    componentLinearAccScale=<0.5 0.5 0.5>
    componentLinearVelScale=<0.5 0.5 0.5>
    gravityOverride {z=-980}
    physicsBodyDefinitions [{
      boundBone {boneName=dyn_headband_straps_a_04}
      boxExtents=<20 30 30>
      constraintSetup {angularLimitsMax=<0 5 15>, angularLimitsMin=<0 -15 0>, angularTarget=<1 0 0>}
      localJointOffset=<5 0 0>
     }]
    simSpaceSettings {maxAngularAcceleration=40, maxAngularVelocity=200, simSpaceAngularAlpha=0.6}
   }
   n22 = Anim Dynamics - Bone: dyn_headband_straps_a_05!(ComponentPose=@n21, Alpha=1, GravityOverride:Vector="0,0,-980", ExternalForce="0,0,0") -> (Pose) {
    angularDampingOverride=1
    angularSpringConstant=8000
    bAngularSpring=true
    bOverrideAngularDamping=true
    bUseGravityOverride=true
    gravityScale=0.35
    simulationSpace=RootRelative
    boundBone {boneName=dyn_headband_straps_a_05}
    componentAppliedLinearAccClamp=<2000 2000 2000>
    componentLinearAccScale=<0.5 0.5 0.5>
    componentLinearVelScale=<0.5 0.5 0.5>
    gravityOverride {z=-980}
    physicsBodyDefinitions [{
      boundBone {boneName=dyn_headband_straps_a_05}
      boxExtents=<20 30 30>
      constraintSetup {angularLimitsMax=<0 5 15>, angularLimitsMin=<0 -15 0>, angularTarget=<1 0 0>}
      localJointOffset=<5 0 0>
     }]
    simSpaceSettings {maxAngularAcceleration=40, maxAngularVelocity=200, simSpaceAngularAlpha=0.6}
   }
   Output Pose(Result=Component To Local(ComponentPose=@n22)) {layerGroup=None}
   n23 = Control Rig!(
    Source=Copy Pose From Mesh(SourceMeshComponent="Get Face Mesh "() -> ("Face Mesh ":Skeletal Mesh Component Object Reference)) {bUseAttachedParent=true}
    Alpha
   ) -> (Pose) {controlRigAssetReference {blueprintRigClass=&/M_MED_RaisinHunt_Rag_FaceAcc_CtrlRig.M_MED_RaisinHunt_Rag_FaceAcc_CtrlRig_C}}
   n24 = RigidBody(
    ComponentPose=Local To Component(LocalPose=@n23)
    Alpha=Get RBANAlpha() -> (RBANAlpha:double)
    OverrideWorldGravity="0,0,0"
    SimSpaceSettings=(WorldAlpha=0,VelocityScaleZ=1,DampingAlpha=1,MaxLinearVelocity=10000,MaxAngularVelocity=10000,MaxLinearAcceleration=10000,MaxAngularAcceleration=10000,ExternalLinearDragV=(X=0,Y=0,Z=0),ExternalLinearVelocity=(X=0,Y=0,Z=0),ExternalAngularVelocity=(X=0,Y=0,Z=0))
   ) {
    bOverrideWorldGravity=true
    overridePhysicsAsset=&/M_MED_RaisinHunt_Rag_FaceAcc_Physics
    simulationSpace=BaseBoneSpace
    baseBoneRef {boneName=spine_05}
   }
   n25 = "Constraint - dyn_right_hair_a_01 (P)"!(ComponentPose=@n24, Alpha=1, ConstraintWeights_0=1, ConstraintWeights_1:float=0) -> (Pose) {
    alphaScaleBiasClamp {interpSpeedDecreasing=0.5, interpSpeedIncreasing=0.5}
    boneToModify {boneName=dyn_right_hair_a_01}
    constraintSetup [{targetBone {boneName=head}}]
    constraintWeights [1,0]
   }
   n26 = "Constraint - dyn_right_hair_a_02 (PPPP)"!(
    ComponentPose=@n25
    Alpha=1
    ConstraintWeights_0=0.7686
    ConstraintWeights_1:float=0.1892
    ConstraintWeights_2:float=0.0262
    ConstraintWeights_3:float=0.0159
   ) -> (Pose) {
    alphaScaleBiasClamp {interpSpeedDecreasing=0.5, interpSpeedIncreasing=0.5}
    boneToModify {boneName=dyn_right_hair_a_02}
    constraintSetup [{targetBone {boneName=head}}, {targetBone {boneName=neck_02}}, {targetBone {boneName=spine_05}}, {targetBone {boneName=neck_01}}]
    constraintWeights [1,1,1,1]
   }
   n27 = "Constraint - dyn_right_hair_a_03 (PPPPP)"!(
    ComponentPose=@n26
    Alpha=0.8
    ConstraintWeights_0=0.6625
    ConstraintWeights_1:float=0.1486
    ConstraintWeights_2:float=0.1133
    ConstraintWeights_3:float=0.0636
    ConstraintWeights_4:float=0.0119
   ) -> (Pose) {
    alphaScaleBiasClamp {interpSpeedDecreasing=0.5, interpSpeedIncreasing=0.5}
    boneToModify {boneName=dyn_right_hair_a_03}
    constraintSetup [
     {targetBone {boneName=spine_05}}
     {targetBone {boneName=neck_02}}
     {targetBone {boneName=clavicle_r}}
     {targetBone {boneName=neck_01}}
     {targetBone {boneName=head}}
    ]
    constraintWeights [1,1,1,1,1]
   }
   n28 = "Constraint - dyn_right_hair_a_05 (PPPP)"!(
    ComponentPose=@n27
    Alpha=0.5
    ConstraintWeights_0=0.5982
    ConstraintWeights_1:float=0.2843
    ConstraintWeights_2:float=0.0757
    ConstraintWeights_3:float=0.0418
   ) -> (Pose) {
    alphaScaleBiasClamp {interpSpeedDecreasing=0.5, interpSpeedIncreasing=0.5}
    boneToModify {boneName=dyn_right_hair_a_05}
    constraintSetup [{targetBone {boneName=head}}, {targetBone {boneName=neck_02}}, {targetBone {boneName=spine_05}}, {targetBone {boneName=neck_01}}]
    constraintWeights [1,1,1,1]
   }
   n29 = Trail controller - Bone: dyn_headband_straps_a_05(ComponentPose=@n28, Alpha=0) {
    chainLength=4
    relaxationSpeedScale=1.5
    rotationLimits [{limitMax=<180 180 180>}, {limitMax=<180 180 180>}, {limitMax=<180 180 180>}, {limitMax=<180 180 180>}]
    rotationOffsets [<0 0 0>, <0 0 0>, <0 0 0>, <0 0 0>]
    trailBone {boneName=dyn_headband_straps_a_05}
    alphaBoolBlend=~
   }
   Save cached pose 'BAse_01'(Pose=Component To Local(ComponentPose=@n29)) {cacheName=BAse_01}
   Reroute Node(InputPin:Wildcard) -> (OutputPin:Wildcard)
  ]}
 EventGraph {nodes [
   Event OnCharacterCustomizationLoaded()
   > n1 = Cast To PlayerPawn_Athena(Object=TryGetPawnOwner())
   > "Set Face Mesh "("Face Mesh ":Skeletal Mesh Component Object Reference=GetSkeletalMeshForPartType(self=@n1, PartType=Head)) -> (Output_Get:Skeletal Mesh Component Object Reference)
   Event BlueprintInitializeAnimation()
   > Set Skeletal_Mesh(Skeletal_Mesh:Skeletal Mesh Component Object Reference=GetOwnerSkeletalMeshComponent()) -> (Output_Get:Skeletal Mesh Component Object Reference)
   Event BlueprintUpdateAnimation()
   > n2 = Sequence!(execute) -> (then_0, then_1, then_2:Exec, then_3:Exec)
   > n3 = "Gravity Override State Function (Parameter Asset) (Nativized)"(
    CharacterPartAnimInstance=Self-Reference()
    GameTimeInSeconds=GetGameTimeInSeconds()
    Parameters=&/M_MED_RaisinHunt_Rag_FaceAcc_Parameter_01
   )
   > Set GravityOverride(GravityOverride=@n3)
   > Set RBAN Settings Out(RBAN Settings Out=@n3.OutRigidBodyAnimNodeParameters)
   n4 = NOT Boolean(A=Get bIsPlayingEmote())
   n5 = NOT Boolean(A=Get BeltForBackpack())
   n6 = Get bIsInFrontEnd()
   > Set FEBackpackANim(FEBackpackANim:double="To Float (Boolean)"(InBool=AND Boolean!(A=@n5, B=@n6, C:bool=@n4) -> (ReturnValue))) -> (Output_Get:double)
   n7 = Cast To FortPlayerPawn(execute=@n2.then_2, Object=GetOwningActor())
   > n8 = "(Deprecated) Get Variant Channel Active Index For Item"!(
    execute
    Item:Athena Cosmetic Item Definition Object Reference=/BRCosmetics/Athena/Items/Cosmetics/Characters/Character_RaisinHunt_Rag
    Loadout:"Fort Athena Loadout Structure (by ref)"=Get CosmeticLoadout!(self:Fort Player Pawn Object Reference=@n7) -> (CosmeticLoadout:Fort Athena Loadout Structure)
    Channel:Gameplay Tag Structure=(TagName="Cosmetics.Variant.Channel.ProfileLoadout")
    VariantList:Array of Gameplay Tag Structures=Make Array("[0]"=(TagName="Cosmetics.Variant.Property.Stage1"))
    bErrorIfChannelMissing:bool=false
    bErrorIfTagNotInArray:bool=false
   ) -> (then, Index:int=0, bIndexIsValid:bool, bFoundChannel:bool)
   > n9 = Branch(Condition=@n8.bIndexIsValid)
   > Set BeltForBackpack(BeltForBackpack:bool=false) -> (Output_Get:bool)
   Set BeltForBackpack(execute=@n9.else, BeltForBackpack=true)
   n10 = Cast To FortPlayerPawn(execute=@n2.then_3, Object=GetOwningActor())
   > n11 = "(Deprecated) Get Variant Channel Active Index For Item"(
    Item=&Character_RaisinHunt_Rag
    Loadout=Get CosmeticLoadout(self=@n10)
    Channel=(TagName="Cosmetics.Variant.Channel.Parts")
    VariantList=Make Array("[0]"=(TagName="Cosmetics.Variant.Property.Stage2"))
    bErrorIfChannelMissing=false
    bErrorIfTagNotInArray=false
   ) -> (Index=0)
   > n12 = Branch(Condition=@n11.bIndexIsValid)
   > Set TankTopOnOff(TankTopOnOff:bool=true) -> (Output_Get:bool)
   Set TankTopOnOff(execute=@n12.else, TankTopOnOff=false)
   n13 = OR Boolean!(A=Get bIsSkydiveFloating(), B=Get bIsSkydiveDiveMode(), C:bool=Get bIsParachuteOpened()) -> (ReturnValue)
   n14 = Get bIsJumping()
   n15 = Get bIsFalling()
   n16 = Branch(execute=@n2.then_1, Condition=AND Boolean!(A=OR Boolean!(A=@n14, B=@n15) -> (ReturnValue), B=NOT Boolean(A=@n13)) -> (ReturnValue))
   > Set IsJumping(IsJumping:double=1) -> (Output_Get:double)
   Set IsJumping(execute=@n16.else, IsJumping=0)
   > n17 = Branch(Condition=Get bIsSkydiveFloating())
   > Set RBANAlpha(RBANAlpha:double=0) -> (Output_Get:double)
   Set RBANAlpha(execute=@n17.else, RBANAlpha=1)
   Get bIsTacticalSprinting()
  ]}
]
variables [
 BeltForBackpack {type=boolean}
 "Face Mesh " {type=SkeletalMeshComponent}
 FEBackpackANim {type=number}
 GravityOverride {type=Vector}
 IsJumping {type=number}
 RBAN Settings Out {type=FortRigidBodyAnimNodeParameters}
 RBANAlpha {type=number}
 Skeletal_Mesh {type=SkeletalMeshComponent}
 TankTopOnOff {type=boolean}
]
